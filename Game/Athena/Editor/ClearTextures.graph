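# ClearTextures is an editor utility Blueprint that overwrites a configurable list of Texture2D assets with the contents of the ResetCaptureTarget render target.
# On BeginExecution it loops over TexturesToReset, converting the render target into each texture with the editor-only conversion node.
# Each texture is then saved immediately, whether dirty or not, and the task finishes once the loop completes.
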EditorUtilityBlueprint /Game/Athena/Editor/ClearTextures
graphs [EventGraph {nodes [
   Event BeginExecution()
   > n1 = For Each Loop(Array:Array of Texture 2D Object References=Get TexturesToReset() -> (TexturesToReset:Array of Texture 2D Object References)) -> (Array Element:Texture 2D Object Reference)
   FinishExecutingTask(execute=@n1.Completed)
   ConvertRenderTargetToTexture2DEditorOnly(execute=@n1.LoopBody, RenderTarget=&/ResetCaptureTarget, Texture=@n1)
   > SaveAssetsWhilePlaying!(
    execute
    AssetsToSave:Array of Object References=Make Array("[0]":Object Reference=@n1) -> (Array:Array of Object References)
    bOnlyIfIsDirty:bool=false
   ) -> (then, ReturnValue:bool)
  ]}]
variables [TexturesToReset {type="array<Texture2D>", default []}]
