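# Editor utility task that automates Battle Royale minimap captures for a configured map and playlist, with Living World disabled.
# It chains async child tasks that clear target textures, load the map, capture discovered and undiscovered views at 4k and optional 8k, then stop PIE.
# A flag adds the final Apollo minimap textures to the 4k targets.
# The discovered 8k capture also exports as map-named JPG and PNG into Gauntlet test data.
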EditorUtilityBlueprint /Game/Athena/Editor/Capture_Minimap_Automation_V2
graphs [
 EventGraph {nodes [
   Event BeginExecution()
   > SetDeveloperSettingLivingWorldEnabled!(execute, bEnabled:bool=false) -> (then)
   > SetDeveloperSettingPlaylist!(
    execute
    Playlist:Fort Playlist Athena Soft Object Reference=Get CapturePlaylist() -> (CapturePlaylist:Fort Playlist Athena Soft Object Reference)
   ) -> (then)
   > n1 = Construct Object from Class!(
    execute
    Class=&/ClearTextures.ClearTextures_C
    self=Self-Reference()
    TexturesToReset:Array of Texture 2D Object References=Get Textures_4k_Discovered() -> (Textures_4k_Discovered:Array of Texture 2D Object References)
   ) -> (then, ReturnValue:Clear Textures Object Reference)
   > n2 = RegisterAndExecuteTask!(execute:Exec, Task:Editor Utility Task Object Reference=@n1, OptionalParentTask:Editor Utility Task Object Reference=Self-Reference()) -> (then:Exec, OnFinished:Exec, Task:Editor Utility Task Object Reference) {is=title K2Node_AsyncAction}
   n3 = Construct Object from Class(
    execute=@n2.OnFinished
    Class=&ClearTextures_C
    self=Self-Reference()
    TexturesToReset=Get Textures_4k_Undiscovered() -> (Textures_4k_Undiscovered:Array of Texture 2D Object References)
   )
   > n4 = RegisterAndExecuteTask(Task=@n3, OptionalParentTask=Self-Reference())
   n5 = Branch(execute=@n4.OnFinished, Condition=Get Capture8k() -> (Capture8k:bool))
   > n6 = Construct Object from Class(
    Class=&ClearTextures_C
    self=Self-Reference()
    TexturesToReset=Get Textures_8k_Discovered() -> (Textures_8k_Discovered:Array of Texture 2D Object References)
   )
   > n7 = RegisterAndExecuteTask(Task=@n6, OptionalParentTask=Self-Reference())
   n8 = Branch(execute=[@n7.OnFinished, @n5.else], Condition=Get Capture8k Undiscovered() -> (Capture8k Undiscovered:bool))
   > n9 = Construct Object from Class(
    Class=&ClearTextures_C
    self=Self-Reference()
    TexturesToReset=Get Textures_8k_Undiscovered() -> (Textures_8k_Undiscovered:Array of Texture 2D Object References)
   )
   > n10 = RegisterAndExecuteTask(Task=@n9, OptionalParentTask=Self-Reference())
   n11 = Construct Object from Class!(
    execute=[@n10.OnFinished, @n8.else]
    Class=&/Capture_Minimap.Capture_Minimap_C
    self=Self-Reference()
    LoadEverything:bool=True
    PreserveExistingState:bool=False
    LoadMapOnly:bool=true
    Resolution:int=0
    Map:Soft Object Path Structure=Get CaptureMap() -> (CaptureMap:Soft Object Path Structure)
    ImportTargets:Array of Texture 2D Object References
    Scene Capture Class:Scene Capture 2D Class Reference
   ) -> (then, ReturnValue:Capture Minimap Object Reference)
   > n12 = RegisterAndExecuteTask(Task=@n11, OptionalParentTask=Self-Reference())
   n13 = Construct Object from Class(
    execute=@n12.OnFinished
    Class=&Capture_Minimap_C
    self=Self-Reference()
    LoadEverything=True
    PreserveExistingState=true
    LoadMapOnly=False
    Resolution=4096
    Map=Get CaptureMap()
    ImportTargets=Get4kDiscoveredTextures!() -> (ReturnTextures:Array of Texture 2D Object References)
    Scene Capture Class=Get DiscoveredClass() -> (DiscoveredClass:Scene Capture 2D Class Reference)
   )
   > n14 = RegisterAndExecuteTask(Task=@n13, OptionalParentTask=Self-Reference())
   n15 = Branch(execute=@n14.OnFinished, Condition=Get Capture8k())
   n16 = Get Textures_8k_Discovered()
   > n17 = Construct Object from Class(
    Class=&Capture_Minimap_C
    self=Self-Reference()
    LoadEverything=True
    PreserveExistingState=true
    LoadMapOnly=False
    Resolution=8196
    Map=Get CaptureMap()
    ImportTargets=@n16
    Scene Capture Class=Get DiscoveredClass()
   )
   > n18 = RegisterAndExecuteTask(Task=@n17, OptionalParentTask=Self-Reference())
   n19 = Get CaptureMap()
   n20 = Replace(
    SourceString=Split!(InPath:String=BreakSoftObjectPath(InSoftObjectPath=@n19)) -> (PathPart:String, FilenamePart:String, ExtensionPart:String).FilenamePart
    From=_terrain
   )
   n21 = "Get (a copy)"(Array:Array of Texture 2D Object References=@n16) -> (Output:Texture 2D Object Reference)
   n22 = ExportImageAsync!(
    execute:Exec=@n18.OnFinished
    Texture:Texture Object Reference=@n21
    OutputFile:String=Append!(A="..\\..\\..\\FortniteGame\\Test\\Gauntlet\\Data\\Images\\", B=@n20, C:String=MiniMap.jpg) -> (ReturnValue)
    Quality:int=20
   ) -> (then:Exec, Complete:Exec, bSuccess:bool=~) {is=title K2Node_AsyncAction}
   n23 = ExportImageAsync(
    execute=@n22.Complete
    Texture=@n21
    OutputFile=Append(A="..\\..\\..\\FortniteGame\\Test\\Gauntlet\\Data\\Images\\", B=@n20, C=MiniMap.png)
    Quality=100
   ) -> (bSuccess=~)
   n24 = Construct Object from Class(
    execute=[@n15.else, @n23.Complete]
    Class=&Capture_Minimap_C
    self=Self-Reference()
    LoadEverything=True
    PreserveExistingState=true
    LoadMapOnly=False
    Resolution=4096
    Map=Get CaptureMap()
    ImportTargets=Get4kUndiscoveredTextures!() -> (ReturnTextures:Array of Texture 2D Object References)
    Scene Capture Class=Get UndiscoveredClass() -> (UndiscoveredClass:Scene Capture 2D Class Reference)
   )
   > n25 = RegisterAndExecuteTask(Task=@n24, OptionalParentTask=Self-Reference())
   n26 = Branch(execute=@n25.OnFinished, Condition=Get Capture8k Undiscovered())
   > n27 = Construct Object from Class(
    Class=&Capture_Minimap_C
    self=Self-Reference()
    LoadEverything=True
    PreserveExistingState=true
    LoadMapOnly=False
    Resolution=8192
    Map=Get CaptureMap()
    ImportTargets=Get Textures_8k_Undiscovered()
    Scene Capture Class=Get UndiscoveredClass()
   )
   > n28 = RegisterAndExecuteTask(Task=@n27, OptionalParentTask=Self-Reference())
   n29 = Construct Object from Class!(execute=[@n26.else, @n28.OnFinished], Class=&/StopPIE.StopPIE_C, self=Self-Reference()) -> (then, ReturnValue:Stop PIE Object Reference)
   > n30 = RegisterAndExecuteTask(Task=@n29, OptionalParentTask=Self-Reference())
   FinishExecutingTask(execute=@n30.OnFinished)
  ]}
 Get4kDiscoveredTextures {nodes [
   Get4kDiscoveredTextures() {is=title K2Node_FunctionEntry}
   > n1 = Branch(Condition=Get CaptureToFinalTargets() -> (CaptureToFinalTargets:bool))
   n2 = Get Textures_4k_Discovered()
   > n3 = Add(
    TargetArray:Array of Texture 2D Object References=@n2
    NewItem:"Texture 2D Object Reference (by ref)"=/Game/Athena/Apollo/Maps/UI/Apollo_Terrain_Minimap
   ) {is=title K2Node_CallArrayFunction}
   Return Node!(execute=[@n3.then, @n1.else], ReturnTextures:Array of Texture 2D Object References=@n2)
  ]}
 Get4kUndiscoveredTextures {nodes [
   Get4kUndiscoveredTextures() {is=title K2Node_FunctionEntry}
   > n1 = Branch(Condition=Get CaptureToFinalTargets())
   n2 = Get Textures_4k_Undiscovered()
   > n3 = Add(TargetArray=@n2, NewItem=/BRCosmetics/Athena/Apollo/Maps/UI/Apollo_Terrain_Minimap_Undiscovered)
   Return Node(execute=[@n3.then, @n1.else], ReturnTextures=@n2)
  ]}
]
variables [
 Capture8k {type=boolean, default=true}
 Capture8k Undiscovered {type=boolean}
 CaptureMap {type=SoftObjectPath, default=/Game/Athena/Helios/Maps/Helios_Terrain}
 CapturePlaylist {type=FortPlaylistAthena, default=null}
 CaptureToFinalTargets {type=boolean}
 DiscoveredClass {type=Class, default=&/SceneCapture_Discovered.SceneCapture_Discovered_C}
 Textures_4k_Discovered {type="array<Texture2D>", default [/Game/Athena/Apollo/Maps/UI/Automated_Captures/Minimap_4096_Automated]}
 Textures_4k_Undiscovered {type="array<Texture2D>", default [/Game/Athena/Apollo/Maps/UI/Automated_Captures/Minimap_Undiscovered_4096_Automated]}
 Textures_8k_Discovered {type="array<Texture2D>", default [/Game/Athena/Apollo/Maps/UI/Automated_Captures/Minimap_8192_Automated]}
 Textures_8k_Undiscovered {type="array<Texture2D>", default [null]}
 UndiscoveredClass {type=Class, default=&/SceneCapture_Undiscovered.SceneCapture_Undiscovered_C}
]
