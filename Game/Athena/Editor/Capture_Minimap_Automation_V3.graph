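# Editor utility task that automates discovered and undiscovered minimap texture capture for a Fortnite Athena map.
# The command-line map name selects a config table row supplying map, playlist and texture paths; flags control 8K and final-target output.
# Child tasks run sequentially on completion callbacks: clear textures, load the map, capture each variant, stop PIE.
# The 8K discovered texture also exports as JPG and PNG to Gauntlet test data.
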EditorUtilityBlueprint /Game/Athena/Editor/Capture_Minimap_Automation_V3
graphs [
 EventGraph {nodes [
   Event BeginExecution()
   > SetDeveloperSettingLivingWorldEnabled!(execute, bEnabled:bool=false) -> (then)
   > n1 = GetCommandLine!(execute) -> (then, ReturnValue:String)
   > n2 = GetMapName!(execute, self, CommandLineString:String=@n1, ArgumentToSearch:String="-mapname=") -> (then, MapName:String)
   > SetTaskConfiguration!(execute, self, MapName:String=@n2) -> (then)
   > SetDeveloperSettingPlaylist!(
    execute
    Playlist:Fort Playlist Athena Soft Object Reference=Get CapturePlaylist() -> (CapturePlaylist:Fort Playlist Athena Soft Object Reference)
   ) -> (then)
   > n3 = Construct Object from Class!(
    execute
    Class=&/ClearTextures.ClearTextures_C
    self=Self-Reference()
    TexturesToReset:Array of Texture 2D Object References=Get Textures_4k_Discovered() -> (Textures_4k_Discovered:Array of Texture 2D Object References)
   ) -> (then, ReturnValue:Clear Textures Object Reference)
   > n4 = RegisterAndExecuteTask!(execute:Exec, Task:Editor Utility Task Object Reference=@n3, OptionalParentTask:Editor Utility Task Object Reference=Self-Reference()) -> (then:Exec, OnFinished:Exec, Task:Editor Utility Task Object Reference) {is=title K2Node_AsyncAction}
   n5 = Construct Object from Class(
    execute=@n4.OnFinished
    Class=&ClearTextures_C
    self=Self-Reference()
    TexturesToReset=Get Textures_4k_Undiscovered() -> (Textures_4k_Undiscovered:Array of Texture 2D Object References)
   )
   > n6 = RegisterAndExecuteTask(Task=@n5, OptionalParentTask=Self-Reference())
   n7 = Branch(execute=@n6.OnFinished, Condition=Get Capture8k() -> (Capture8k:bool))
   > n8 = Construct Object from Class(
    Class=&ClearTextures_C
    self=Self-Reference()
    TexturesToReset=Get Textures_8k_Discovered() -> (Textures_8k_Discovered:Array of Texture 2D Object References)
   )
   > n9 = RegisterAndExecuteTask(Task=@n8, OptionalParentTask=Self-Reference())
   n10 = Branch(execute=[@n9.OnFinished, @n7.else], Condition=Get Capture8k Undiscovered() -> (Capture8k Undiscovered:bool))
   > n11 = Construct Object from Class(
    Class=&ClearTextures_C
    self=Self-Reference()
    TexturesToReset=Get Textures_8k_Undiscovered() -> (Textures_8k_Undiscovered:Array of Texture 2D Object References)
   )
   > n12 = RegisterAndExecuteTask(Task=@n11, OptionalParentTask=Self-Reference())
   n13 = Construct Object from Class!(
    execute=[@n12.OnFinished, @n10.else]
    Class=&/Capture_Minimap.Capture_Minimap_C
    self=Self-Reference()
    LoadEverything:bool=True
    PreserveExistingState:bool=False
    LoadMapOnly:bool=true
    Resolution:int=0
    Map:Soft Object Path Structure=Get CaptureMap() -> (CaptureMap:Soft Object Path Structure)
    ImportTargets:Array of Texture 2D Object References
    Scene Capture Class:Scene Capture 2D Class Reference
   ) -> (then, ReturnValue:Capture Minimap Object Reference)
   > n14 = RegisterAndExecuteTask(Task=@n13, OptionalParentTask=Self-Reference())
   PreCaptureSetup(execute=@n14.OnFinished)
   > n15 = Construct Object from Class(
    Class=&Capture_Minimap_C
    self=Self-Reference()
    LoadEverything=True
    PreserveExistingState=true
    LoadMapOnly=False
    Resolution=4096
    Map=Get CaptureMap()
    ImportTargets=Get4kDiscoveredTextures!() -> (ReturnTextures:Array of Texture 2D Object References)
    Scene Capture Class=Get DiscoveredClass() -> (DiscoveredClass:Scene Capture 2D Class Reference)
   )
   > n16 = RegisterAndExecuteTask(Task=@n15, OptionalParentTask=Self-Reference())
   n17 = Branch(execute=@n16.OnFinished, Condition=Get Capture8k())
   > PrintString(InString=Starting 8K Capture, bPrintToScreen=false)
   > n18 = Construct Object from Class(
    Class=&Capture_Minimap_C
    self=Self-Reference()
    LoadEverything=True
    PreserveExistingState=true
    LoadMapOnly=False
    Resolution=8196
    Map=Get CaptureMap()
    ImportTargets=Get Textures_8k_Discovered()
    Scene Capture Class=Get DiscoveredClass()
   )
   > n19 = RegisterAndExecuteTask(Task=@n18, OptionalParentTask=Self-Reference())
   n20 = "Get (a copy)"(Array:Array of Texture 2D Object References=Get Textures_8k_Discovered()) -> (Output:Texture 2D Object Reference)
   Is Valid(exec=@n19.OnFinished, InputObject=@n20)
   n21 = Split!(InPath:String=BreakSoftObjectPath(InSoftObjectPath=Get CaptureMap())) -> (PathPart:String, FilenamePart:String, ExtensionPart:String)
   > n22 = ExportImageAsync!(
    execute:Exec
    Texture:Texture Object Reference=@n20
    OutputFile:String=Append!(A="..\\..\\..\\FortniteGame\\Test\\Gauntlet\\Data\\Images\\", B=@n21.FilenamePart, C:String=_MiniMap.jpg) -> (ReturnValue)
    Quality:int=20
   ) -> (then:Exec, Complete:Exec, bSuccess:bool=~) {is=title K2Node_AsyncAction}
   n23 = ExportImageAsync(
    execute=@n22.Complete
    Texture=@n20
    OutputFile=Append(A="..\\..\\..\\FortniteGame\\Test\\Gauntlet\\Data\\Images\\", B=@n21.FilenamePart, C=_MiniMap.png)
    Quality=100
   ) -> (bSuccess=~)
   n24 = Construct Object from Class(
    execute=[@n17.else, @n23.Complete]
    Class=&Capture_Minimap_C
    self=Self-Reference()
    LoadEverything=True
    PreserveExistingState=true
    LoadMapOnly=False
    Resolution=4096
    Map=Get CaptureMap()
    ImportTargets=Get4kUndiscoveredTextures!() -> (ReturnTextures:Array of Texture 2D Object References)
    Scene Capture Class=Get UndiscoveredClass() -> (UndiscoveredClass:Scene Capture 2D Class Reference)
   )
   > n25 = RegisterAndExecuteTask(Task=@n24, OptionalParentTask=Self-Reference())
   n26 = Branch(execute=@n25.OnFinished, Condition=Get Capture8k Undiscovered())
   > n27 = Construct Object from Class(
    Class=&Capture_Minimap_C
    self=Self-Reference()
    LoadEverything=True
    PreserveExistingState=true
    LoadMapOnly=False
    Resolution=8192
    Map=Get CaptureMap()
    ImportTargets=Get Textures_8k_Undiscovered()
    Scene Capture Class=Get UndiscoveredClass()
   )
   > n28 = RegisterAndExecuteTask(Task=@n27, OptionalParentTask=Self-Reference())
   n29 = Construct Object from Class!(execute=[@n26.else, @n28.OnFinished], Class=&/StopPIE.StopPIE_C, self=Self-Reference()) -> (then, ReturnValue:Stop PIE Object Reference)
   > n30 = RegisterAndExecuteTask(Task=@n29, OptionalParentTask=Self-Reference())
   FinishExecutingTask(execute=@n30.OnFinished)
   Replace(SourceString=@n21.FilenamePart, From=_terrain)
   Reroute Node(InputPin:Wildcard) -> (OutputPin:Wildcard)
  ]}
 FindCommandLineParam {nodes [
   n1 = FindCommandLineParam!() -> (then, CommandLine:String, ParamToFind:String) {is=title K2Node_FunctionEntry}
   > Return Node!(execute, HasParam:bool="integer >= integer"(A=FindSubstring(SearchIn=@n1, Substring=@n1.ParamToFind)))
  ]}
 Get4kDiscoveredTextures {nodes [
   Get4kDiscoveredTextures() {is=title K2Node_FunctionEntry}
   > n1 = Branch(Condition=Get CaptureToFinalTargets() -> (CaptureToFinalTargets:bool))
   n2 = Get Textures_4k_Discovered()
   > n3 = Add(
    TargetArray:Array of Texture 2D Object References=@n2
    NewItem:"Texture 2D Object Reference (by ref)"=/Game/Athena/Apollo/Maps/UI/Apollo_Terrain_Minimap
   ) {is=title K2Node_CallArrayFunction}
   Return Node!(execute=[@n3.then, @n1.else], ReturnTextures:Array of Texture 2D Object References=@n2)
  ]}
 Get4kUndiscoveredTextures {nodes [
   n1 = Get Textures_4k_Undiscovered()
   n2 = Add(TargetArray=@n1, NewItem=/BRCosmetics/Athena/Apollo/Maps/UI/Apollo_Terrain_Minimap_Undiscovered)
   Return Node(execute=[@n2.then, @n3.then, @n3.else], ReturnTextures=@n1)
   Get4kUndiscoveredTextures() {is=title K2Node_FunctionEntry}
   > n3 = Branch(Condition=Get CaptureToFinalTargets())
  ]}
 GetMapName {nodes [
   n1 = GetMapName!() -> (then, CommandLineString:String, ArgumentToSearch:String) {is=title K2Node_FunctionEntry}
   n2 = int + int(A=FindSubstring(SearchIn=@n1, Substring=@n1.ArgumentToSearch), B=Len(S=@n1.ArgumentToSearch))
   n3 = FindSubstring(SearchIn=@n1, Substring=" ", StartPosition=@n2)
   > Return Node!(
    execute
    MapName:String=GetSubstring!(SourceString:String=@n1, StartIndex:int=@n2, Length:int=int - int(A=@n3, B=@n2)) -> (ReturnValue:String)
   )
  ]}
 Load4kTextures {nodes [
   n1 = Load4kTextures!() -> (then, DiscoveredTexturesPaths:Array of Strings, UndiscoveredTexturesPaths:Array of Strings) {is=title K2Node_FunctionEntry}
   > Clear(TargetArray:Array of Texture 2D Object References=Get Textures_4k_Discovered())
   > Clear(TargetArray=Get Textures_4k_Undiscovered())
   > n2 = For Each Loop(Array=@n1)
   > n3 = LoadAsset(AssetPath=@n2)
   > n4 = Cast To Texture2D(Object=@n3)
   n5 = Get Textures_4k_Discovered()
   > Add(TargetArray=@n5, NewItem=@n4)
   PrintString(execute=@n4.CastFailed, InString=Append!(A="NO ASSET FOUND FOR OBJECT PATH: ", B=@n2) -> (ReturnValue), bPrintToScreen=false)
   n6 = CheckoutLoadedAssets(execute=@n2.Completed, AssetsToCheckout=@n5)
   > PrintString(InString=Append(A="Texture checkout operation success: ", B="To String (Boolean)"(InBool=@n6)))
   > n7 = For Each Loop(Array=@n1.UndiscoveredTexturesPaths)
   > n8 = LoadAsset(AssetPath=@n7)
   > n9 = Cast To Texture2D(Object=@n8)
   n10 = Get Textures_4k_Undiscovered()
   > Add(TargetArray=@n10, NewItem=@n9)
   PrintString(execute=@n9.CastFailed, InString=Append(A="NO ASSET FOUND FOR OBJECT PATH: ", B=@n7), bPrintToScreen=false)
   n11 = CheckoutLoadedAssets(execute=@n7.Completed, AssetsToCheckout=@n10)
   > PrintString(InString=Append(A="Texture checkout operation success: ", B="To String (Boolean)"(InBool=@n11)))
   > Return Node!(execute)
  ]}
 Load8kTextures {nodes [
   n1 = Load8kTextures!() -> (then, DiscoveredTexturesPaths:Array of Strings, UndiscoveredTexturesPaths:Array of Strings) {is=title K2Node_FunctionEntry}
   > Clear(TargetArray=Get Textures_8k_Discovered())
   > Clear(TargetArray=Get Textures_8k_Undiscovered())
   > n2 = For Each Loop(Array=@n1)
   > n3 = LoadAsset(AssetPath=@n2)
   > n4 = Cast To Texture2D(Object=@n3)
   n5 = Get Textures_8k_Discovered()
   > Add(TargetArray=@n5, NewItem=@n4)
   PrintString(execute=@n4.CastFailed, InString=Append(A="NO ASSET FOUND FOR OBJECT PATH: ", B=@n2), bPrintToScreen=false)
   n6 = CheckoutLoadedAssets(execute=@n2.Completed, AssetsToCheckout=@n5)
   > PrintString(InString=Append(A="Texture checkout operation success: ", B="To String (Boolean)"(InBool=@n6)))
   > n7 = For Each Loop(Array=@n1.UndiscoveredTexturesPaths)
   > n8 = LoadAsset(AssetPath=@n7)
   > n9 = Cast To Texture2D(Object=@n8)
   n10 = Get Textures_8k_Undiscovered()
   > Add(TargetArray=@n10, NewItem=@n9)
   PrintString(execute=@n9.CastFailed, InString=Append(A="NO ASSET FOUND FOR OBJECT PATH: ", B=@n7), bPrintToScreen=false)
   n11 = CheckoutLoadedAssets(execute=@n7.Completed, AssetsToCheckout=@n10)
   > PrintString(InString=Append(A="Texture checkout operation success: ", B="To String (Boolean)"(InBool=@n11)))
   > Return Node()
  ]}
 PreCaptureSetup {nodes [PreCaptureSetup() {is=title K2Node_FunctionEntry}]}
 SetTaskConfiguration {nodes [
   n1 = SetTaskConfiguration!() -> (then, MapName:String) {is=title K2Node_FunctionEntry}
   > n2 = Get Data Table Row!(
    execute:Exec
    DataTable:Data Table Object Reference=Get ConfigurationDataTable() -> (ConfigurationDataTable:Data Table Object Reference)
    RowName:Name=String To Name(InString=@n1)
   ) -> (then:Exec, RowNotFound:Exec, ReturnValue:Capture Minimap Configuration Structure)
   n3 = Break Capture Minimap Configuration!(CaptureMinimapConfiguration:"Capture Minimap Configuration Structure (by ref)"=@n2) -> (
    CaptureMap_19_#1:String
    Textures4kDiscovered_20_#2:Array of Strings
    Textures4kUndiscovered_21_#3:Array of Strings
    Textures8kDiscovered_22_#4:Array of Strings
    Textures8kUndiscovered_23_#5:Array of Strings
    CapturePlaylist_26_#6:String
   ) {is=title K2Node_BreakStruct}
   n4 = LoadAsset(AssetPath=@n3.CapturePlaylist_26_#6)
   n5 = GetCommandLine(execute=[@n4.then, @n15.then, @n13.then, @n12.then])
   > n6 = FindCommandLineParam!(execute, self, CommandLine:String=@n5, ParamToFind:String=-capture8k) -> (then, HasParam:bool) {is=title K2Node_CallFunction}
   > n7 = Set Capture8k(Capture8k:bool=@n6) -> (Output_Get:bool)
   > PrintString(InString=Append(A="Capture 8k value set to: ", B="To String (Boolean)"(InBool=@n7)))
   > n8 = GetCommandLine()
   > n9 = FindCommandLineParam(CommandLine=@n8, ParamToFind=-capturefinaltargets)
   > n10 = Set CaptureToFinalTargets(CaptureToFinalTargets:bool=@n9) -> (Output_Get:bool)
   > PrintString(InString=Append(A="Capture to final targets value set to: ", B="To String (Boolean)"(InBool=@n10)))
   > Return Node()
   n11 = Cast To FortPlaylistAthena!(execute, Object=@n4) -> (then, CastFailed, AsFort Playlist Athena:Fort Playlist Athena Object Reference) {is=title K2Node_DynamicCast}
   > n12 = Set CapturePlaylist(CapturePlaylist:Fort Playlist Athena Soft Object Reference=Make Soft Reference!(Input:Fort Playlist Athena Object Reference=@n11) -> (Output:Fort Playlist Athena Soft Object Reference)) -> (Output_Get:Fort Playlist Athena Soft Object Reference)
   n13 = PrintString(execute=@n11.CastFailed, InString=CAPTURE TASK: NO PLAYLIST ASSET LOADED, bPrintToScreen=false)
   n14 = Reroute Node(InputPin:Exec) -> (OutputPin:Exec)
   Load4kTextures!(
    execute=[@n18.then, @n14.OutputPin]
    self
    DiscoveredTexturesPaths:Array of Strings=@n3.Textures4kDiscovered_20_#2
    UndiscoveredTexturesPaths:Array of Strings=@n3.Textures4kUndiscovered_21_#3
   ) -> (then) {is=title K2Node_CallFunction}
   > n15 = Load8kTextures!(
    execute
    self
    DiscoveredTexturesPaths:Array of Strings=@n3.Textures8kDiscovered_22_#4
    UndiscoveredTexturesPaths:Array of Strings=@n3.Textures8kUndiscovered_23_#5
   ) -> (then) {is=title K2Node_CallFunction}
   Return Node(execute=@n16.then)
   n16 = PrintString(execute=@n2.RowNotFound, InString=Configuration row not found in capture task, bPrintToScreen=false)
   n17 = Set MapSoftObjPath(execute=@n2.then, MapSoftObjPath:String=@n3) -> (Output_Get:String)
   n18 = PrintString(execute=@n20.then, InString=Append(A=Configuration path:, B=@n17), bPrintToScreen=false)
   n19 = MakeSoftObjectPath!(PathString:String=@n17) -> (ReturnValue:Soft Object Path Structure)
   n20 = Set CaptureMap(execute=@n17.then, CaptureMap:Soft Object Path Structure=@n19) -> (Output_Get:Soft Object Path Structure)
  ]}
]
variables [
 Capture8k {type=boolean, default=true}
 Capture8k Undiscovered {type=boolean, default=true}
 CaptureMap {type=SoftObjectPath, default=/Game/Athena/Helios/Maps/Helios_Terrain}
 CapturePlaylist {type=FortPlaylistAthena, default=null}
 CaptureToFinalTargets {type=boolean}
 ConfigurationDataTable {type=DataTable, default=&/Capture_Minimap_configuration_DataTable}
 DiscoveredClass {type=Class, default=&/SceneCapture_Discovered.SceneCapture_Discovered_C}
 MapSoftObjPath {type=string, default=""}
 Textures_4k_Discovered {type="array<Texture2D>", default [null]}
 Textures_4k_Undiscovered {type="array<Texture2D>", default [null]}
 Textures_8k_Discovered {type="array<Texture2D>", default [null]}
 Textures_8k_Undiscovered {type="array<Texture2D>", default [null]}
 UndiscoveredClass {type=Class, default=&/SceneCapture_Undiscovered.SceneCapture_Undiscovered_C}
]
