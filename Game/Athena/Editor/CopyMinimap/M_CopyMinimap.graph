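# M_CopyMinimap is an unlit material in an editor CopyMinimap folder, apparently for copying minimap captures.
# Its graph splits a color into R, G and B, converts each channel with sRGBToLinear, then recombines them with MakeFloat3.
# A Minimap texture parameter defaults to the automated 4096 Apollo capture.
# The file shows no links from that texture into the conversion or from the result to a material output.
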Material /Game/Athena/Editor/CopyMinimap/M_CopyMinimap
graph {
 properties {pixelDepthOffsetMode=PDOM_Legacy, shadingModel=MSM_Unlit, usedShadingModels=Unlit}
 nodes [
  n1 = MaterialFunctionCall() {materialFunction=&BreakOutFloat3Components}
  MaterialFunctionCall(
   X=MaterialFunctionCall!(sRGB=@n1.R) {materialFunction=/Engine/Functions/Engine_MaterialFunctions01/ImageAdjustment/sRGBToLinear}.Linear
   Y=MaterialFunctionCall(sRGB=@n1.G) {materialFunction=&sRGBToLinear}.Linear
   Z=MaterialFunctionCall(sRGB=@n1.B) {materialFunction=&sRGBToLinear}.Linear
  ) {materialFunction=&MakeFloat3}
  TextureSampleParameter2D() {parameterName=Minimap, texture=/Game/Athena/Apollo/Maps/UI/Automated_Captures/Minimap_4096_Automated}
 ]
}
