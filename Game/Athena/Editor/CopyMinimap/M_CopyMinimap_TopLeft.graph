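# Unlit editor material for copying the top-left part of the automated 8192 Athena minimap capture, as its name indicates.
# Texture coordinates tiled at half scale on both axes address one quadrant of the capture.
# The engine TextureCropping function takes those UVs, the fixed capture texture and two constant corner vectors.
# A separate Minimap texture parameter defaults to the same capture but does not feed the crop.
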Material /Game/Athena/Editor/CopyMinimap/M_CopyMinimap_TopLeft
graph {
 properties {pixelDepthOffsetMode=PDOM_Legacy, shadingModel=MSM_Unlit, usedShadingModels=Unlit}
 nodes [
  MaterialFunctionCall!(
   UVs=TextureCoordinate() {uTiling=0.5, vTiling=0.5}
   TextureIn=TextureObject() {texture=/Game/Athena/Apollo/Maps/UI/Automated_Captures/Minimap_8192_Automated}
   UpperLeftCorner=Constant2Vector() {g=1}
   LowerRightCorner=Constant2Vector() {g=1}
   ExportFloat4
  ) {materialFunction=/Engine/Functions/Engine_MaterialFunctions01/Texturing/TextureCropping}
  TextureSampleParameter2D() {parameterName=Minimap, texture=&Minimap_8192_Automated}
 ]
}
