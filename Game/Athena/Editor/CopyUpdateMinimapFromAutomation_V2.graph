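# Editor utility widget that updates the Apollo terrain minimap and undiscovered minimap textures from automated captures at 8k, 4k or 2k.
# Copy buttons select a material and target texture, draw into the MinimapCopy render target, then convert it into that texture with alpha-free compression.
# Other buttons check out, save and check in each map.
# Undiscovered copies target a /BRCosmetics path, while their save buttons use /Game.
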EditorUtilityWidgetBlueprint /Game/Athena/Editor/CopyUpdateMinimapFromAutomation_V2
graphs [EventGraph {nodes [
   "On Button Clicked (IconTextButton)"!() -> (OutputDelegate, then, Button:Common Button Legacy Object Reference) {is=title K2Node_ComponentBoundEvent}
   > n1 = Set MaterialtoRender(MaterialtoRender:Material Interface Object Reference=/Game/Athena/Editor/CopyMinimap/M_CopyMinimap_8k) -> (Output_Get:Material Interface Object Reference)
   n2 = Set Texture(execute=[@n1.then, @n5.then, @n6.then], Texture:Texture 2D Object Reference=/Game/Athena/Apollo/Maps/UI/Apollo_Terrain_Minimap) -> (Output_Get:Texture 2D Object Reference)
   n3 = Get Texture Render Target() -> (Texture Render Target:Texture Render Target 2D Object Reference)
   DrawMaterialToRenderTarget!(
    execute=[@n8.then, @n2.then]
    WorldContextObject:Object Reference
    TextureRenderTarget:Texture Render Target 2D Object Reference=@n3
    Material:Material Interface Object Reference=Get MaterialtoRender() -> (MaterialtoRender:Material Interface Object Reference)
   ) -> (then)
   n4 = Get Texture() -> (Texture:Texture 2D Object Reference)
   > ConvertRenderTargetToTexture2DEditorOnly(RenderTarget=@n3, Texture=@n4)
   > Set CompressionNoAlpha!(execute, CompressionNoAlpha:bool=true, self:Texture Object Reference=@n4) -> (then, Output_Get:bool)
   "On Button Clicked (IconTextButton_9)"!() -> (OutputDelegate, then, Button:Common Button Legacy Object Reference) {is=title K2Node_ComponentBoundEvent}
   > CheckoutAsset!(execute, AssetToCheckout:String=/Game/Athena/Apollo/Maps/UI/Apollo_Terrain_Minimap_Undiscovered) -> (then, ReturnValue:bool)
   > SaveAsset!(execute, AssetToSave:String=&Apollo_Terrain_Minimap_Undiscovered, bOnlyIfIsDirty:bool=false) -> (then, ReturnValue:bool)
   > CheckInFile!(
    execute
    InFile:String=&Apollo_Terrain_Minimap_Undiscovered
    InDescription:String="##JIRA update  Undiscovered Minimap from Automated Capture"
    bSilent:bool=false
    bKeepCheckedOut:bool=false
   ) -> (then, ReturnValue:bool)
   "On Button Clicked (IconTextButton_1)"!() -> (OutputDelegate, then, Button:Common Button Legacy Object Reference) {is=title K2Node_ComponentBoundEvent}
   > CheckoutAsset(AssetToCheckout=&Apollo_Terrain_Minimap)
   > SaveAsset(AssetToSave=&Apollo_Terrain_Minimap, bOnlyIfIsDirty=false)
   "On Button Clicked (IconTextButton_2)"!() -> (OutputDelegate, then, Button:Common Button Legacy Object Reference) {is=title K2Node_ComponentBoundEvent}
   > n5 = Set MaterialtoRender(MaterialtoRender=/Game/Athena/Editor/CopyMinimap/M_CopyMinimap_TopLeft)
   "On Button Clicked (IconTextButton_3)"!() -> (OutputDelegate, then, Button:Common Button Legacy Object Reference) {is=title K2Node_ComponentBoundEvent}
   > n6 = Set MaterialtoRender(MaterialtoRender=/Game/Athena/Editor/CopyMinimap/M_CopyMinimap_2k)
   "On Button Clicked (IconTextButton_4)"!() -> (OutputDelegate, then, Button:Common Button Legacy Object Reference) {is=title K2Node_ComponentBoundEvent}
   > n7 = Set MaterialtoRender(MaterialtoRender=/Game/Athena/Editor/CopyMinimap/M_CopyMinimap_Undiscovered8k)
   n8 = Set Texture(execute=[@n10.then, @n7.then, @n9.then], Texture=/BRCosmetics/Athena/Apollo/Maps/UI/Apollo_Terrain_Minimap_Undiscovered)
   "On Button Clicked (IconTextButton_5)"!() -> (OutputDelegate, then, Button:Common Button Legacy Object Reference) {is=title K2Node_ComponentBoundEvent}
   > n9 = Set MaterialtoRender(MaterialtoRender=/Game/Athena/Editor/CopyMinimap/M_CopyMinimap_Undiscovered4k)
   "On Button Clicked (IconTextButton_6)"!() -> (OutputDelegate, then, Button:Common Button Legacy Object Reference) {is=title K2Node_ComponentBoundEvent}
   > n10 = Set MaterialtoRender(MaterialtoRender=/Game/Athena/Editor/CopyMinimap/M_CopyMinimap_Undiscovered2k)
   "On Button Clicked (IconTextButton_7)"!() -> (OutputDelegate, then, Button:Common Button Legacy Object Reference) {is=title K2Node_ComponentBoundEvent}
   > CheckoutAsset(AssetToCheckout=/Game/Athena/Apollo/Maps/UI/Apollo_Terrain_Minimap_Undiscovered)
   > SaveAsset(AssetToSave=&Apollo_Terrain_Minimap_Undiscovered, bOnlyIfIsDirty=false)
   "On Button Clicked (IconTextButton_8)"!() -> (OutputDelegate, then, Button:Common Button Legacy Object Reference) {is=title K2Node_ComponentBoundEvent}
   > CheckoutAsset(AssetToCheckout=&Apollo_Terrain_Minimap)
   > SaveAsset(AssetToSave=&Apollo_Terrain_Minimap, bOnlyIfIsDirty=false)
   > CheckInFile(
    InFile=&Apollo_Terrain_Minimap
    InDescription="##JIRA update  Minimap from Automated Capture"
    bSilent=false
    bKeepCheckedOut=false
   )
   Event PreConstruct()
   Event Construct()
   Event Tick()
  ]}]
variables [
 MaterialtoRender {type=MaterialInterface, default=/Game/Athena/Editor/CopyMinimap/M_CopyMinimap}
 Texture {type=Texture2D, default=&Apollo_Terrain_Minimap}
 Texture Render Target {type=TextureRenderTarget2D, default=/Game/Athena/Editor/CopyMinimap/MinimapCopy}
]
widgets {
 info {rootWidgetClass=&CanvasPanel, widgetCount=17}
 widgets [
  CanvasPanel_0 {widgetClassPath=&CanvasPanel, widget=CanvasPanel_0}
  IconTextButton {
   widgetClassPath=&IconTextButton
   bIsVariable=true
   parent=CanvasPanel_0
   widget=IconTextButton
   properties {
    buttonText="Copy Minimap  from 8k (Best)"
    onButtonBaseClicked=(&self:WidgetTree.IconTextButton.HandleOnButtonClicked)
    onButtonBaseDoubleClicked=(&self:WidgetTree.IconTextButton.HandleOnButtonDoubleClicked)
    onButtonBaseHovered=(&self:WidgetTree.IconTextButton.HandleOnButtonHovered)
    onButtonBaseUnhovered=(&self:WidgetTree.IconTextButton.HandleOnButtonUnhovered)
    onSelectedChangedBase=(&self:WidgetTree.IconTextButton.HandleOnSelectedChanged)
   }
   slot {
    class=CanvasPanelSlot
    properties {layoutData {alignment=<0 0>, anchors {maximum=<0 0>, minimum=<0 0>}, offsets=<ltrb 24 24 472.0778 80.1708>}}
   }
  }
  IconTextButton_4 {
   widgetClassPath=&IconTextButton
   bIsVariable=true
   parent=CanvasPanel_0
   widget=IconTextButton_4
   properties {
    buttonText="Copy Undiscovered  from 8k (Best)"
    onButtonBaseClicked=(&self:WidgetTree.IconTextButton_4.HandleOnButtonClicked)
    onButtonBaseDoubleClicked=(&self:WidgetTree.IconTextButton_4.HandleOnButtonDoubleClicked)
    onButtonBaseHovered=(&self:WidgetTree.IconTextButton_4.HandleOnButtonHovered)
    onButtonBaseUnhovered=(&self:WidgetTree.IconTextButton_4.HandleOnButtonUnhovered)
    onSelectedChangedBase=(&self:WidgetTree.IconTextButton_4.HandleOnSelectedChanged)
   }
   slot {
    class=CanvasPanelSlot
    properties {layoutData {alignment=<0 0>, anchors {maximum=<0 0>, minimum=<0 0>}, offsets=<ltrb 688 32 509.8934 80.1708>}}
   }
  }
  IconTextButton_5 {
   widgetClassPath=&IconTextButton
   bIsVariable=true
   parent=CanvasPanel_0
   widget=IconTextButton_5
   properties {
    buttonText="Copy Undiscovered  from 4k (Good)"
    onButtonBaseClicked=(&self:WidgetTree.IconTextButton_5.HandleOnButtonClicked)
    onButtonBaseDoubleClicked=(&self:WidgetTree.IconTextButton_5.HandleOnButtonDoubleClicked)
    onButtonBaseHovered=(&self:WidgetTree.IconTextButton_5.HandleOnButtonHovered)
    onButtonBaseUnhovered=(&self:WidgetTree.IconTextButton_5.HandleOnButtonUnhovered)
    onSelectedChangedBase=(&self:WidgetTree.IconTextButton_5.HandleOnSelectedChanged)
   }
   slot {
    class=CanvasPanelSlot
    properties {layoutData {alignment=<0 0>, anchors {maximum=<0 0>, minimum=<0 0>}, offsets=<ltrb 692 192 509.8934 80.1708>}}
   }
  }
  IconTextButton_6 {
   widgetClassPath=&IconTextButton
   bIsVariable=true
   parent=CanvasPanel_0
   widget=IconTextButton_6
   properties {
    buttonText="Copy Undiscovered  from 2k (Bad)"
    onButtonBaseClicked=(&self:WidgetTree.IconTextButton_6.HandleOnButtonClicked)
    onButtonBaseDoubleClicked=(&self:WidgetTree.IconTextButton_6.HandleOnButtonDoubleClicked)
    onButtonBaseHovered=(&self:WidgetTree.IconTextButton_6.HandleOnButtonHovered)
    onButtonBaseUnhovered=(&self:WidgetTree.IconTextButton_6.HandleOnButtonUnhovered)
    onSelectedChangedBase=(&self:WidgetTree.IconTextButton_6.HandleOnSelectedChanged)
   }
   slot {
    class=CanvasPanelSlot
    properties {layoutData {alignment=<0 0>, anchors {maximum=<0 0>, minimum=<0 0>}, offsets=<ltrb 696 344 509.8934 80.1708>}}
   }
  }
  IconTextButton_2 {
   widgetClassPath=&IconTextButton
   bIsVariable=true
   parent=CanvasPanel_0
   widget=IconTextButton_2
   properties {
    buttonText="Copy Minimap  from 4k (Good)"
    onButtonBaseClicked=(&self:WidgetTree.IconTextButton_2.HandleOnButtonClicked)
    onButtonBaseDoubleClicked=(&self:WidgetTree.IconTextButton_2.HandleOnButtonDoubleClicked)
    onButtonBaseHovered=(&self:WidgetTree.IconTextButton_2.HandleOnButtonHovered)
    onButtonBaseUnhovered=(&self:WidgetTree.IconTextButton_2.HandleOnButtonUnhovered)
    onSelectedChangedBase=(&self:WidgetTree.IconTextButton_2.HandleOnSelectedChanged)
   }
   slot {
    class=CanvasPanelSlot
    properties {layoutData {alignment=<0 0>, anchors {maximum=<0 0>, minimum=<0 0>}, offsets=<ltrb 20 196 468.6511 80.1708>}}
   }
  }
  IconTextButton_3 {
   widgetClassPath=&IconTextButton
   bIsVariable=true
   parent=CanvasPanel_0
   widget=IconTextButton_3
   properties {
    buttonText="Copy Minimap  from 2k (Bad)"
    onButtonBaseClicked=(&self:WidgetTree.IconTextButton_3.HandleOnButtonClicked)
    onButtonBaseDoubleClicked=(&self:WidgetTree.IconTextButton_3.HandleOnButtonDoubleClicked)
    onButtonBaseHovered=(&self:WidgetTree.IconTextButton_3.HandleOnButtonHovered)
    onButtonBaseUnhovered=(&self:WidgetTree.IconTextButton_3.HandleOnButtonUnhovered)
    onSelectedChangedBase=(&self:WidgetTree.IconTextButton_3.HandleOnSelectedChanged)
   }
   slot {
    class=CanvasPanelSlot
    properties {layoutData {alignment=<0 0>, anchors {maximum=<0 0>, minimum=<0 0>}, offsets=<ltrb 20 352 466.9377 80.1708>}}
   }
  }
  IconTextButton_1 {
   widgetClassPath=&IconTextButton
   bIsVariable=true
   parent=CanvasPanel_0
   widget=IconTextButton_1
   properties {
    buttonText=Checkout and Save Minimap
    onButtonBaseClicked=(&self:WidgetTree.IconTextButton_1.HandleOnButtonClicked)
    onButtonBaseDoubleClicked=(&self:WidgetTree.IconTextButton_1.HandleOnButtonDoubleClicked)
    onButtonBaseHovered=(&self:WidgetTree.IconTextButton_1.HandleOnButtonHovered)
    onButtonBaseUnhovered=(&self:WidgetTree.IconTextButton_1.HandleOnButtonUnhovered)
    onSelectedChangedBase=(&self:WidgetTree.IconTextButton_1.HandleOnSelectedChanged)
    colorAndOpacity=<rgba 0.2838 1 0.2025 1>
   }
   slot {
    class=CanvasPanelSlot
    properties {layoutData {alignment=<0 0>, anchors {maximum=<0 0>, minimum=<0 0>}, offsets=<ltrb 32 484 443.6028 83.7962>}}
   }
  }
  IconTextButton_8 {
   widgetClassPath=&IconTextButton
   bIsVariable=true
   parent=CanvasPanel_0
   widget=IconTextButton_8
   properties {
    buttonText=Checkin Minimap
    onButtonBaseClicked=(&self:WidgetTree.IconTextButton_8.HandleOnButtonClicked)
    onButtonBaseDoubleClicked=(&self:WidgetTree.IconTextButton_8.HandleOnButtonDoubleClicked)
    onButtonBaseHovered=(&self:WidgetTree.IconTextButton_8.HandleOnButtonHovered)
    onButtonBaseUnhovered=(&self:WidgetTree.IconTextButton_8.HandleOnButtonUnhovered)
    onSelectedChangedBase=(&self:WidgetTree.IconTextButton_8.HandleOnSelectedChanged)
    colorAndOpacity=<rgba 0.2838 1 0.2025 1>
   }
   slot {
    class=CanvasPanelSlot
    properties {layoutData {alignment=<0 0>, anchors {maximum=<0 0>, minimum=<0 0>}, offsets=<ltrb 32 584 443.6028 83.7962>}}
   }
  }
  IconTextButton_7 {
   widgetClassPath=&IconTextButton
   bIsVariable=true
   parent=CanvasPanel_0
   widget=IconTextButton_7
   properties {
    buttonText=Checkout and Save Undiscovered Map
    onButtonBaseClicked=(&self:WidgetTree.IconTextButton_7.HandleOnButtonClicked)
    onButtonBaseDoubleClicked=(&self:WidgetTree.IconTextButton_7.HandleOnButtonDoubleClicked)
    onButtonBaseHovered=(&self:WidgetTree.IconTextButton_7.HandleOnButtonHovered)
    onButtonBaseUnhovered=(&self:WidgetTree.IconTextButton_7.HandleOnButtonUnhovered)
    onSelectedChangedBase=(&self:WidgetTree.IconTextButton_7.HandleOnSelectedChanged)
    colorAndOpacity=<rgba 0.2838 1 0.2025 1>
   }
   slot {
    class=CanvasPanelSlot
    properties {layoutData {alignment=<0 0>, anchors {maximum=<0 0>, minimum=<0 0>}, offsets=<ltrb 704 496 608.714 83.7962>}}
   }
  }
  IconTextButton_9 {
   widgetClassPath=&IconTextButton
   bIsVariable=true
   parent=CanvasPanel_0
   widget=IconTextButton_9
   properties {
    buttonText=Checkin Undiscovered Map
    onButtonBaseClicked=(&self:WidgetTree.IconTextButton_9.HandleOnButtonClicked)
    onButtonBaseDoubleClicked=(&self:WidgetTree.IconTextButton_9.HandleOnButtonDoubleClicked)
    onButtonBaseHovered=(&self:WidgetTree.IconTextButton_9.HandleOnButtonHovered)
    onButtonBaseUnhovered=(&self:WidgetTree.IconTextButton_9.HandleOnButtonUnhovered)
    onSelectedChangedBase=(&self:WidgetTree.IconTextButton_9.HandleOnSelectedChanged)
    colorAndOpacity=<rgba 0.2838 1 0.2025 1>
   }
   slot {
    class=CanvasPanelSlot
    properties {layoutData {alignment=<0 0>, anchors {maximum=<0 0>, minimum=<0 0>}, offsets=<ltrb 704 592 608.714 83.7962>}}
   }
  }
  Image_1 {
   widgetClassPath=&Image
   bIsVariable=true
   parent=CanvasPanel_0
   widget=Image_1
   properties {brush {
     bIsDynamicallyLoaded=false
     drawAs=Image
     imageType=NoImage
     mirroring=NoMirror
     resourceName=None
     resourceObject=/Game/Athena/Apollo/Maps/UI/Automated_Captures/Minimap_2048_Automated
     tiling=NoTile
     imageSize=<32 32>
     margin=<ltrb 0 0 0 0>
     outlineSettings {
      bUseBrushTransparency=false
      roundingType=HalfHeightRadius
      width=0
      color {colorUseRule=UseColor_Specified, specifiedColor=<rgba 0 0 0 0>}
      cornerRadii=<0 0 0 0>
     }
     tintColor {colorUseRule=UseColor_Specified, specifiedColor=<rgba 1 1 1 1>}
     uVRegion {bIsValid=false, max=<0 0>, min=<0 0>}
    }}
   slot {class=CanvasPanelSlot, properties {layoutData {alignment=<0 0>, anchors {maximum=<0 0>, minimum=<0 0>}, offsets=<ltrb 516 328 128 128>}}}
  }
  Image_2 {
   widgetClassPath=&Image
   bIsVariable=true
   parent=CanvasPanel_0
   widget=Image_2
   properties {brush {
     bIsDynamicallyLoaded=false
     drawAs=Image
     imageType=NoImage
     mirroring=NoMirror
     resourceName=None
     resourceObject=/Game/Athena/Apollo/Maps/UI/Automated_Captures/Minimap_4096_Automated
     tiling=NoTile
     imageSize=<32 32>
     margin=<ltrb 0 0 0 0>
     outlineSettings {
      bUseBrushTransparency=false
      roundingType=HalfHeightRadius
      width=0
      color {colorUseRule=UseColor_Specified, specifiedColor=<rgba 0 0 0 0>}
      cornerRadii=<0 0 0 0>
     }
     tintColor {colorUseRule=UseColor_Specified, specifiedColor=<rgba 1 1 1 1>}
     uVRegion {bIsValid=false, max=<0 0>, min=<0 0>}
    }}
   slot {class=CanvasPanelSlot, properties {layoutData {alignment=<0 0>, anchors {maximum=<0 0>, minimum=<0 0>}, offsets=<ltrb 516 168 128 128>}}}
  }
  Image_3 {
   widgetClassPath=&Image
   bIsVariable=true
   parent=CanvasPanel_0
   widget=Image_3
   properties {brush {
     bIsDynamicallyLoaded=false
     drawAs=Image
     imageType=NoImage
     mirroring=NoMirror
     resourceName=None
     resourceObject=/Game/Athena/Apollo/Maps/UI/Automated_Captures/Minimap_8192_Automated
     tiling=NoTile
     imageSize=<32 32>
     margin=<ltrb 0 0 0 0>
     outlineSettings {
      bUseBrushTransparency=false
      roundingType=HalfHeightRadius
      width=0
      color {colorUseRule=UseColor_Specified, specifiedColor=<rgba 0 0 0 0>}
      cornerRadii=<0 0 0 0>
     }
     tintColor {colorUseRule=UseColor_Specified, specifiedColor=<rgba 1 1 1 1>}
     uVRegion {bIsValid=false, max=<0 0>, min=<0 0>}
    }}
   slot {class=CanvasPanelSlot, properties {layoutData {alignment=<0 0>, anchors {maximum=<0 0>, minimum=<0 0>}, offsets=<ltrb 516 16 128 128>}}}
  }
  Image_4 {
   widgetClassPath=&Image
   bIsVariable=true
   parent=CanvasPanel_0
   widget=Image_4
   properties {brush {
     bIsDynamicallyLoaded=false
     drawAs=Image
     imageType=NoImage
     mirroring=NoMirror
     resourceName=None
     resourceObject=/Game/Athena/Apollo/Maps/UI/Automated_Captures/Minimap_Undiscovered_2048_Automated
     tiling=NoTile
     imageSize=<32 32>
     margin=<ltrb 0 0 0 0>
     outlineSettings {
      bUseBrushTransparency=false
      roundingType=HalfHeightRadius
      width=0
      color {colorUseRule=UseColor_Specified, specifiedColor=<rgba 0 0 0 0>}
      cornerRadii=<0 0 0 0>
     }
     tintColor {colorUseRule=UseColor_Specified, specifiedColor=<rgba 1 1 1 1>}
     uVRegion {bIsValid=false, max=<0 0>, min=<0 0>}
    }}
   slot {class=CanvasPanelSlot, properties {layoutData {alignment=<0 0>, anchors {maximum=<0 0>, minimum=<0 0>}, offsets=<ltrb 1248 332 128 128>}}}
  }
  Image_5 {
   widgetClassPath=&Image
   bIsVariable=true
   parent=CanvasPanel_0
   widget=Image_5
   properties {brush {
     bIsDynamicallyLoaded=false
     drawAs=Image
     imageType=NoImage
     mirroring=NoMirror
     resourceName=None
     resourceObject=/Game/Athena/Apollo/Maps/UI/Automated_Captures/Minimap_Undiscovered_4096_Automated
     tiling=NoTile
     imageSize=<32 32>
     margin=<ltrb 0 0 0 0>
     outlineSettings {
      bUseBrushTransparency=false
      roundingType=HalfHeightRadius
      width=0
      color {colorUseRule=UseColor_Specified, specifiedColor=<rgba 0 0 0 0>}
      cornerRadii=<0 0 0 0>
     }
     tintColor {colorUseRule=UseColor_Specified, specifiedColor=<rgba 1 1 1 1>}
     uVRegion {bIsValid=false, max=<0 0>, min=<0 0>}
    }}
   slot {class=CanvasPanelSlot, properties {layoutData {alignment=<0 0>, anchors {maximum=<0 0>, minimum=<0 0>}, offsets=<ltrb 1248 176 128 128>}}}
  }
  Image_6 {
   widgetClassPath=&Image
   bIsVariable=true
   parent=CanvasPanel_0
   widget=Image_6
   properties {brush {
     bIsDynamicallyLoaded=false
     drawAs=Image
     imageType=NoImage
     mirroring=NoMirror
     resourceName=None
     resourceObject=/Game/Athena/Apollo/Maps/UI/Automated_Captures/Minimap_Undiscovered_8192_Automated
     tiling=NoTile
     imageSize=<32 32>
     margin=<ltrb 0 0 0 0>
     outlineSettings {
      bUseBrushTransparency=false
      roundingType=HalfHeightRadius
      width=0
      color {colorUseRule=UseColor_Specified, specifiedColor=<rgba 0 0 0 0>}
      cornerRadii=<0 0 0 0>
     }
     tintColor {colorUseRule=UseColor_Specified, specifiedColor=<rgba 1 1 1 1>}
     uVRegion {bIsValid=false, max=<0 0>, min=<0 0>}
    }}
   slot {class=CanvasPanelSlot, properties {layoutData {alignment=<0 0>, anchors {maximum=<0 0>, minimum=<0 0>}, offsets=<ltrb 1248 16 128 128>}}}
  }
 ]
}
