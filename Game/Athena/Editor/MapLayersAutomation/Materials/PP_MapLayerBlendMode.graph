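# Post-process material in the map layers automation editor tools that composites a layer texture over a base image, by default an automated minimap capture.
# An Operation Index drives a branched switch in a Custom node, choosing among 18 Photoshop-style blend modes computed in parallel.
# The saturated result is lerped onto the base.
# Parameters derive the layer mask from alpha or selected channels, with inversion and opacity, and tint or flat-fill its color.
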Material /Game/Athena/Editor/MapLayersAutomation/Materials/PP_MapLayerBlendMode
graph {
 properties {materialDomain=MD_PostProcess, pixelDepthOffsetMode=PDOM_Legacy, usedShadingModels=Default Lit}
 nodes [
  MaterialExpressionScalarParameter_1 = ScalarParameter() {parameterName=Operation Index}
  MaterialExpressionNamedRerouteUsage_11 = NamedRerouteUsage()
  n1 = NamedRerouteUsage()
  n2 = NamedRerouteUsage()
  MaterialExpressionMaterialFunctionCall_11 = MaterialFunctionCall!(Blend=@n1, Base=@n2) {materialFunction=/Engine/Functions/Engine_MaterialFunctions03/Blends/Blend_Darken}
  MaterialExpressionMultiply_0 = Multiply(A=@n2, B=@n1)
  MaterialExpressionMaterialFunctionCall_9 = MaterialFunctionCall!(Base=@n2, Blend=@n1) {materialFunction=/Engine/Functions/Engine_MaterialFunctions03/Blends/Blend_ColorBurn}
  MaterialExpressionMaterialFunctionCall_16 = MaterialFunctionCall!(Blend=@n2, Base=@n1) {materialFunction=/Engine/Functions/Engine_MaterialFunctions03/Blends/Blend_LinearBurn}
  n3 = NamedRerouteUsage()
  n4 = NamedRerouteUsage()
  MaterialExpressionMaterialFunctionCall_15 = MaterialFunctionCall!(Blend=@n3, Base=@n4) {materialFunction=/Engine/Functions/Engine_MaterialFunctions03/Blends/Blend_Lighten}
  MaterialExpressionMaterialFunctionCall_21 = MaterialFunctionCall!(Blend=@n3, Base=@n4) {materialFunction=/Engine/Functions/Engine_MaterialFunctions03/Blends/Blend_Screen}
  MaterialExpressionMaterialFunctionCall_10 = MaterialFunctionCall!(Blend=@n3, Base=@n4) {materialFunction=/Engine/Functions/Engine_MaterialFunctions03/Blends/Blend_ColorDodge}
  MaterialExpressionMaterialFunctionCall_17 = MaterialFunctionCall!(Blend=@n3, Base=@n4) {materialFunction=/Engine/Functions/Engine_MaterialFunctions03/Blends/Blend_LinearDodge}
  n5 = NamedRerouteUsage()
  n6 = NamedRerouteUsage()
  MaterialExpressionMaterialFunctionCall_19 = MaterialFunctionCall(Base=@n5, Blend=@n6) {materialFunction=&Blend_Overlay}
  MaterialExpressionMaterialFunctionCall_22 = MaterialFunctionCall!(Base=@n5, Blend=@n6) {materialFunction=/Engine/Functions/Engine_MaterialFunctions03/Blends/Blend_SoftLight}
  MaterialExpressionMaterialFunctionCall_14 = MaterialFunctionCall!(Base=@n5, Blend=@n6) {materialFunction=/Engine/Functions/Engine_MaterialFunctions03/Blends/Blend_HardLight}
  MaterialExpressionMaterialFunctionCall_18 = MaterialFunctionCall!(Base=@n5, Blend=@n6) {materialFunction=/Engine/Functions/Engine_MaterialFunctions03/Blends/Blend_LinearLight}
  MaterialExpressionMaterialFunctionCall_20 = MaterialFunctionCall!(Base=@n5, Blend=@n6) {materialFunction=/Engine/Functions/Engine_MaterialFunctions03/Blends/Blend_PinLight}
  n7 = NamedRerouteUsage()
  n8 = NamedRerouteUsage()
  MaterialExpressionMaterialFunctionCall_12 = MaterialFunctionCall!(Blend=@n7, Base=@n8) {materialFunction=/Engine/Functions/Engine_MaterialFunctions03/Blends/Blend_Difference}
  MaterialExpressionMaterialFunctionCall_13 = MaterialFunctionCall!(Blend=@n7, Base=@n8) {materialFunction=/Engine/Functions/Engine_MaterialFunctions03/Blends/Blend_Exclusion}
  MaterialExpressionSubtract_0 = Subtract(A=@n8)
  MaterialExpressionDivide_0 = Divide(A=@n8, B=@n7)
  n9 = Custom!(
   Index=@MaterialExpressionScalarParameter_1
   Normal=@MaterialExpressionNamedRerouteUsage_11
   Darken=@MaterialExpressionMaterialFunctionCall_11.Result
   Multiply=@MaterialExpressionMultiply_0
   ColorBurn=@MaterialExpressionMaterialFunctionCall_9.Result
   LinearBurn=@MaterialExpressionMaterialFunctionCall_16.Result
   Lighten=@MaterialExpressionMaterialFunctionCall_15.Result
   Screen=@MaterialExpressionMaterialFunctionCall_21.Result
   ColorDodge=@MaterialExpressionMaterialFunctionCall_10.Result
   LinearDodge=@MaterialExpressionMaterialFunctionCall_17.Result
   Overlay=@MaterialExpressionMaterialFunctionCall_19.Result
   SoftLight=@MaterialExpressionMaterialFunctionCall_22.Result
   HardLight=@MaterialExpressionMaterialFunctionCall_14.Result
   LinearLight=@MaterialExpressionMaterialFunctionCall_18.Result
   PinLight=@MaterialExpressionMaterialFunctionCall_20.Result
   Difference=@MaterialExpressionMaterialFunctionCall_12.Result
   Exclusion=@MaterialExpressionMaterialFunctionCall_13.Result
   Subtract=@MaterialExpressionSubtract_0
   Divide=@MaterialExpressionDivide_0
  ) {
   code="int i = Index;\r\n\r\n[branch] switch(i)\r\n{\r\n    case 0:\r\n        return Normal; \r\n    case 1:\r\n        return Darken; \r\n    case 2:\r\n        return Multiply;\r\n    case 3:\r\n        return ColorBurn;\r\n    case 4:\r\n        return LinearBurn;\r\n    case 5:\r\n        return Lighten;\r\n    case 6:\r\n        return Screen;\r\n    case 7:\r\n        return ColorDodge;\r\n    case 8:\r\n        return LinearDodge;\r\n    case 9:\r\n        return Overlay;\r\n    case 10:\r\n        return SoftLight;\r\n    case 11:\r\n        return HardLight;\r\n    case 12:\r\n        return LinearLight;\r\n    case 13:\r\n        return PinLight;\r\n    case 14:\r\n        return Difference;\r\n    case 15:\r\n        return Exclusion;\r\n    case 16:\r\n        return Subtract;\r\n    case 17:\r\n        return Divide; \r\n    default:\r\n        return Normal; \r\n}"
   inputs [
    {
     inputName=Index
     input {expression=&self:MaterialExpressionScalarParameter_1}
    }
    {
     inputName=Normal
     input {expression=&self:MaterialExpressionNamedRerouteUsage_11}
    }
    {
     inputName=Darken
     input {expression=&self:MaterialExpressionMaterialFunctionCall_11}
    }
    {inputName=Multiply, input {expression=&self:MaterialExpressionMultiply_0}}
    {
     inputName=ColorBurn
     input {expression=&self:MaterialExpressionMaterialFunctionCall_9}
    }
    {
     inputName=LinearBurn
     input {expression=&self:MaterialExpressionMaterialFunctionCall_16}
    }
    {
     inputName=Lighten
     input {expression=&self:MaterialExpressionMaterialFunctionCall_15}
    }
    {
     inputName=Screen
     input {expression=&self:MaterialExpressionMaterialFunctionCall_21}
    }
    {
     inputName=ColorDodge
     input {expression=&self:MaterialExpressionMaterialFunctionCall_10}
    }
    {
     inputName=LinearDodge
     input {expression=&self:MaterialExpressionMaterialFunctionCall_17}
    }
    {
     inputName=Overlay
     input {expression=&self:MaterialExpressionMaterialFunctionCall_19}
    }
    {
     inputName=SoftLight
     input {expression=&self:MaterialExpressionMaterialFunctionCall_22}
    }
    {
     inputName=HardLight
     input {expression=&self:MaterialExpressionMaterialFunctionCall_14}
    }
    {
     inputName=LinearLight
     input {expression=&self:MaterialExpressionMaterialFunctionCall_18}
    }
    {
     inputName=PinLight
     input {expression=&self:MaterialExpressionMaterialFunctionCall_20}
    }
    {
     inputName=Difference
     input {expression=&self:MaterialExpressionMaterialFunctionCall_12}
    }
    {
     inputName=Exclusion
     input {expression=&self:MaterialExpressionMaterialFunctionCall_13}
    }
    {inputName=Subtract, input {expression=&self:MaterialExpressionSubtract_0}}
    {inputName=Divide, input {expression=&self:MaterialExpressionDivide_0}}
   ]
  }
  n10 = LinearInterpolate(
   A=NamedRerouteDeclaration(None=TextureSampleParameter2D() {parameterName=Base, texture=/Game/Athena/Apollo/Maps/UI/Automated_Captures/Minimap_4096_Automated}.RGB) {name=Base Image, nodeColor=<rgba 0.9453 1 0 1>}
   B=Saturate(None=@n9)
   Alpha=NamedRerouteUsage()
  )
  MaterialFunctionCall!("Xb1,PS4", "PC-High&&Epic,XSX,PS5"=@n10, "PC-Perf,Android", Switch, "PC-Low(NotPerf)&&Medium"=@n10) {materialFunction=/Game/Effects/Niagara/MasterMaterials/Functions/Common/PlatformSwitch_3D}
  n11 = TextureSampleParameter2D() {parameterName=Layer, texture=/Game/Global/Textures/Default/Masks_and_Gradients/16x}
  NamedRerouteDeclaration(None=@n11.RGB) {name=Layer Image, nodeColor=<rgba 1 0.4922 0 1>}
  n12 = VectorParameter() {parameterName=MaskChannels, defaultValue=<rgba 1 0 0 0>}
  n13 = ScalarParameter() {parameterName=As Mask}
  n14 = LinearInterpolate(A=@n11.A, B=DotProduct(A=@n11.RGBA, B=AppendVector(A=@n12.RGB, B=@n12.RGB)), Alpha=@n13)
  n15 = LinearInterpolate(A=@n14, B=OneMinus(None=@n14), Alpha=ScalarParameter() {parameterName=Invert Mask})
  n16 = VectorParameter() {parameterName=Uniform Color, defaultValue=<rgba 1 0 0.1231 1>}
  NamedRerouteDeclaration(None=Saturate(None=Multiply(A=@n15, B=@n16.A))) {name=Layer Mask, nodeColor=<rgba 1 0.4922 0 1>}
  n17 = NamedRerouteUsage()
  NamedRerouteDeclaration(None=LinearInterpolate(A=Multiply(A=@n16.RGB, B=@n17) {desc=Tint image}, B=@n16.RGB, Alpha=@n13)) {name=Color, nodeColor=<rgba 0.125 1 0 1>}
 ]
}
