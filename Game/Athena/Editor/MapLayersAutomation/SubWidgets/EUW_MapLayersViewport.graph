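# Editor utility viewport sub-widget for map layer automation, showing a pannable, zoomable image with a dynamic instance of UI_DisplayMaterial.
# Pressing an invisible full-size button opens a Tick-driven gate that moves the image by mouse delta divided by Anchor scale, until release.
# Mouse wheel calls ZoomFunction, which doubles or halves the Anchor canvas's render scale.
# Cursor position is stored but unused by zoom; resize, reposition and debug-print nodes are disconnected.
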EditorUtilityWidgetBlueprint /Game/Athena/Editor/MapLayersAutomation/SubWidgets/EUW_MapLayersViewport
graphs [
 EventGraph {nodes [
   "On Pressed (ViewportButton)"() {is=title K2Node_ComponentBoundEvent}
   > n1 = GetMousePositionOnViewport!(execute, WorldContextObject:Object Reference) -> (then, ReturnValue:Vector 2D Structure)
   > Set MousePositionBuffered(MousePositionBuffered:Vector 2D Structure=@n1) -> (Output_Get:Vector 2D Structure)
   n2 = Get ViewportImage() -> (ViewportImage:Image Object Reference)
   > n3 = Set ViewportImagePositionBuffered(ViewportImagePositionBuffered:Vector 2D Structure=GetPosition!(self:Canvas Panel Slot Object Reference=SlotAsCanvasSlot!(Widget:Widget Object Reference=@n2) -> (ReturnValue:Canvas Panel Slot Object Reference)) -> (ReturnValue:Vector 2D Structure)) -> (Output_Get:Vector 2D Structure)
   Gate!(Enter:Exec=@n12.then, Open:Exec=@n3.then, Close:Exec=@n7.then, Toggle:Exec, bStartClosed:bool=true) -> (Exit:Exec) {is=title K2Node_MacroInstance}
   > n4 = GetMousePositionOnViewport()
   n5 = vector2d - vector2d(A:Vector 2D Structure=@n4, B:Vector 2D Structure=Get MousePositionBuffered() -> (MousePositionBuffered:Vector 2D Structure)) -> (ReturnValue:Vector 2D Structure) {is=title K2Node_PromotableOperator}
   n6 = Get!(self:Widget Object Reference=Get Anchor() -> (Anchor:Canvas Panel Object Reference)) -> (
    RenderTransform_Translation:Vector 2D Structure
    RenderTransform_Scale:Vector 2D Structure
    RenderTransform_Shear:Vector 2D Structure
    RenderTransform_Angle:float=~
   )
   > SetPosition!(
    execute
    self:Canvas Panel Slot Object Reference=SlotAsCanvasSlot(Widget=Get ViewportImage())
    InPosition:Vector 2D Structure=vector2d + vector2d(
     A:Vector 2D Structure=vector2d / vector2d(A:Vector 2D Structure=@n5, B:Vector 2D Structure=@n6.RenderTransform_Scale) -> (ReturnValue:Vector 2D Structure) {is=title K2Node_PromotableOperator}
     B:Vector 2D Structure=Get ViewportImagePositionBuffered() -> (ViewportImagePositionBuffered:Vector 2D Structure)
    ) -> (ReturnValue:Vector 2D Structure) {is=title K2Node_PromotableOperator}
   ) -> (then)
   n7 = "On Released (ViewportButton)"() {is=title K2Node_ComponentBoundEvent}
   n8 = Zoom!() -> (OutputDelegate, then, Direction:double=~) {is=title K2Node_CustomEvent}
   > ZoomFunction!(execute, self, Direction:double=@n8.Direction) -> (then)
   Event PreConstruct()
   > n9 = CreateDynamicMaterialInstance(Parent=/Game/Athena/Editor/MapLayersAutomation/Materials/UI_DisplayMaterial, OptionalName=None, CreationFlags=None)
   > n10 = Set Material(Material:Material Instance Dynamic Object Reference=@n9) -> (Output_Get:Material Instance Dynamic Object Reference)
   > SetBrushFromMaterial(self=Get ViewportImage(), Material=@n10)
   n11 = Event Tick()
   > n12 = Set My Geo(My Geo:Geometry Structure=@n11.MyGeometry) -> (Output_Get:Geometry Structure)
   Event Construct()
  ]}
 OnMouseWheel {nodes [
   n1 = OnMouseWheel!() -> (then, MyGeometry:Geometry Structure, MouseEvent:"Pointer Event Structure (by ref)") {is=title K2Node_FunctionEntry}
   > Zoom!(
    execute
    self
    Direction:double=Get Wheel Delta!(Input:"Pointer Event Structure (by ref)"=@n1.MouseEvent) -> (ReturnValue:float) {is=title K2Node_CallFunction}
   ) -> (then) {is=title K2Node_CallFunction}
   > Return Node!(execute, ReturnValue:Event Reply Structure=Handled())
  ]}
 ZoomFunction {nodes [
   n1 = SetSize!(execute, self:Canvas Panel Slot Object Reference=SlotAsCanvasSlot(Widget=Get ViewportImage()), InSize:Vector 2D Structure=(X=1024,Y=1024)) -> (then)
   n2 = Get Mouse Position() -> (Mouse Position:Vector 2D Structure)
   n3 = Get My Geo() -> (My Geo:Geometry Structure)
   PrintString(
    execute=[@n1.then, @n4.then]
    InString="To String (Vector2d)"!(InVec:Vector 2D Structure="Absolute to Local (Coordinates)"!(Geometry:"Geometry Structure (by ref)"=@n3, AbsoluteCoordinate:Vector 2D Structure=@n2) -> (ReturnValue:Vector 2D Structure)) -> (ReturnValue:String)
   )
   n4 = SetPosition(self=SlotAsCanvasSlot(Widget=Get ViewportImage()))
   n5 = ZoomFunction!() -> (then, Direction:double=~) {is=title K2Node_FunctionEntry}
   > n6 = GetMousePositionOnViewport()
   > Set Mouse Position(Mouse Position:Vector 2D Structure=@n6) -> (Output_Get:Vector 2D Structure)
   n7 = "Clamp (Float)"(Value="Sign (Float)"!(A:double=@n5) -> (ReturnValue:double), Min=0, Max=1)
   n8 = Get Anchor()
   n9 = Get(self=@n8) -> (RenderTransform_Angle=~)
   > SetRenderScale!(
    execute
    self:Widget Object Reference=@n8
    Scale:Vector 2D Structure=Select(
     Option 0:Vector 2D Structure=vector2d * vector2d(A:Vector 2D Structure=@n9.RenderTransform_Scale, B:Vector 2D Structure=(X=0.5,Y=0.5)) -> (ReturnValue:Vector 2D Structure) {is=title K2Node_PromotableOperator}
     Option 1:Vector 2D Structure=vector2d * vector2d(A=@n9.RenderTransform_Scale, B=(X=2,Y=2))
     Index:int=Truncate(A=@n7)
    ) -> (ReturnValue:Vector 2D Structure)
   ) -> (then)
   Get RenderTransformPivot!(self:Widget Object Reference=@n8) -> (RenderTransformPivot:Vector 2D Structure)
  ]}
]
variables [
 Material {type=MaterialInstanceDynamic}
 MousePositionBuffered {type=Vector2D, default=<0 0>}
 My Geo {type=Geometry, default {}}
 ViewportImagePositionBuffered {type=Vector2D, default=<0 0>}
 ZoomLevel {type=integer}
]
widgets {
 info {rootWidgetClass=&CanvasPanel, widgetCount=4}
 widgets [
  CanvasPanel_20 {widgetClassPath=&CanvasPanel, widget=CanvasPanel_20, properties {clipping=ClipToBounds}}
  Anchor {
   widgetClassPath=&CanvasPanel
   bIsVariable=true
   parent=CanvasPanel_20
   widget=Anchor
   slot {class=CanvasPanelSlot, properties {layoutData {alignment=<0.5 0.5>, anchors {maximum=<0.5 0.5>, minimum=<0.5 0.5>}, offsets=<ltrb 0 0 0 0>}}}
  }
  ViewportImage {
   widgetClassPath=&Image
   bIsVariable=true
   parent=Anchor
   widget=ViewportImage
   properties {brush {
     bIsDynamicallyLoaded=false
     drawAs=Image
     imageType=FullColor
     mirroring=NoMirror
     resourceName=None
     resourceObject=null
     tiling=NoTile
     imageSize=<32 32>
     margin=<ltrb 0 0 0 0>
     outlineSettings {
      bUseBrushTransparency=false
      roundingType=HalfHeightRadius
      width=0
      color {colorUseRule=UseColor_Specified, specifiedColor=<rgba 0 0 0 0>}
      cornerRadii=<0 0 0 0>
     }
     tintColor {colorUseRule=UseColor_Specified, specifiedColor=<rgba 1 1 1 1>}
     uVRegion {bIsValid=false, max=<0 0>, min=<0 0>}
    }}
   slot {
    class=CanvasPanelSlot
    properties {layoutData {alignment=<0 0>, anchors {maximum=<0.5 0.5>, minimum=<0.5 0.5>}, offsets=<ltrb -512 -512 1024 1024>}}
   }
  }
  ViewportButton {
   widgetClassPath=&EditorUtilityButton
   parent=CanvasPanel_20
   widget=ViewportButton
   properties {widgetStyle {
     clickedSlateSound {resourceObject=null}
     disabled {
      bIsDynamicallyLoaded=false
      drawAs=RoundedBox
      imageType=NoImage
      mirroring=NoMirror
      resourceName=None
      resourceObject=null
      tiling=NoTile
      imageSize=<0 0>
      margin=<ltrb 0 0 0 0>
      outlineSettings {
       bUseBrushTransparency=false
       roundingType=FixedRadius
       width=1
       color {colorUseRule=UseColor_ColorTable, specifiedColor=<rgba 0 0 0 0>}
       cornerRadii=<4 4 4 4>
      }
      tintColor {colorUseRule=UseColor_Specified, specifiedColor=<rgba 0.0395 0.0395 0.0395 0>}
      uVRegion {bIsValid=false, max=<0 0>, min=<0 0>}
     }
     disabledForeground {colorUseRule=UseColor_ColorTable, specifiedColor=<rgba 0 0 0 0>}
     hovered {
      bIsDynamicallyLoaded=false
      drawAs=RoundedBox
      imageType=NoImage
      mirroring=NoMirror
      resourceName=None
      resourceObject=null
      tiling=NoTile
      imageSize=<32 32>
      margin=<ltrb 0 0 0 0>
      outlineSettings {
       bUseBrushTransparency=false
       roundingType=FixedRadius
       width=1
       color {colorUseRule=UseColor_ColorTable, specifiedColor=<rgba 0 0 0 0>}
       cornerRadii=<4 4 4 4>
      }
      tintColor {colorUseRule=UseColor_Specified, specifiedColor=<rgba 0.0953 0.0953 0.0953 0>}
      uVRegion {bIsValid=false, max=<0 0>, min=<0 0>}
     }
     hoveredForeground {colorUseRule=UseColor_ColorTable, specifiedColor=<rgba 0 0 0 0>}
     hoveredSlateSound {resourceObject=null}
     normal {
      bIsDynamicallyLoaded=false
      drawAs=RoundedBox
      imageType=NoImage
      mirroring=NoMirror
      resourceName=None
      resourceObject=null
      tiling=NoTile
      imageSize=<32 32>
      margin=<ltrb 0 0 0 0>
      outlineSettings {
       bUseBrushTransparency=false
       roundingType=FixedRadius
       width=1
       color {colorUseRule=UseColor_ColorTable, specifiedColor=<rgba 0 0 0 0>}
       cornerRadii=<4 4 4 4>
      }
      tintColor {colorUseRule=UseColor_Specified, specifiedColor=<rgba 0.0395 0.0395 0.0395 0>}
      uVRegion {bIsValid=false, max=<0 0>, min=<0 0>}
     }
     normalForeground {colorUseRule=UseColor_ColorTable, specifiedColor=<rgba 0 0 0 0>}
     normalPadding=<ltrb 12 1.5 12 1.5>
     pressed {
      bIsDynamicallyLoaded=false
      drawAs=RoundedBox
      imageType=NoImage
      mirroring=NoMirror
      resourceName=None
      resourceObject=null
      tiling=NoTile
      imageSize=<32 32>
      margin=<ltrb 0 0 0 0>
      outlineSettings {
       bUseBrushTransparency=false
       roundingType=FixedRadius
       width=1
       color {colorUseRule=UseColor_ColorTable, specifiedColor=<rgba 0 0 0 0>}
       cornerRadii=<4 4 4 4>
      }
      tintColor {colorUseRule=UseColor_Specified, specifiedColor=<rgba 0.0284 0.0284 0.0284 0>}
      uVRegion {bIsValid=false, max=<0 0>, min=<0 0>}
     }
     pressedForeground {colorUseRule=UseColor_ColorTable, specifiedColor=<rgba 0 0 0 0>}
     pressedPadding=<ltrb 12 2.5 12 0.5>
     pressedSlateSound {resourceObject=null}
    }}
   slot {class=CanvasPanelSlot, properties {layoutData {alignment=<0 0>, anchors {maximum=<1 1>, minimum=<0 0>}, offsets=<ltrb 0 0 0 0>}}}
  }
 ]
}
