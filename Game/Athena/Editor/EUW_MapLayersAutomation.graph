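# Editor utility widget that composites Athena minimap textures from a stack of editable blend layers.
# Each render draws a size-selected CopyMinimap material into a render target, then applies visible layers through PP_MapLayerBlendMode dynamic materials, previewed in the viewport.
# Layers load from Table_MapLayerBatchOperations and list as entry widgets in reverse order.
# Save rewrites that table via a built CSV string and converts the target into Terrain_Minimap_Automation.
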EditorUtilityWidgetBlueprint /Game/Athena/Editor/EUW_MapLayersAutomation
graphs [
 EventGraph {nodes [
   "On Clicked (AddLayerBtn)"() {is=title K2Node_ComponentBoundEvent}
   > Add(
    TargetArray:Array of Struct Map Layer Automation Params Structures=Get Layers() -> (Layers:Array of Struct Map Layer Automation Params Structures)
    NewItem:"Struct Map Layer Automation Params Structure (by ref)"=Make Struct Map Layer Automation Params!(
     LayerName_17_#1:String=New Layer
     Operation_2_#2:Enum_MapBlendLayers Enum=NewEnumerator0
     Texture_19_#3:Texture 2D Object Reference
     UniformColor_12_#4:Linear Color Structure=(R=1,G=1,B=1,A=1)
     AsMask_20_#5:bool=False
     InvertMask_14_#6:bool=False
     Visibility_22_#7:bool=False
    ) -> (Struct_MapLayerAutomationParams:Struct Map Layer Automation Params Structure) {is=title K2Node_MakeStruct}
   ) {is=title K2Node_CallArrayFunction}
   > Construct Layers List()
   n1 = "On Clicked (LoadBtn)"() {is=title K2Node_ComponentBoundEvent}
   Load(execute=[@n1.then, @n5.then])
   > Construct Layers List()
   "On Clicked (SaveBtn)"() {is=title K2Node_ComponentBoundEvent}
   > Save()
   > ConvertRenderTargetToTexture2DEditorOnly(
    RenderTarget=Get MapRenderTarget() -> (MapRenderTarget:Texture Render Target 2D Object Reference)
    Texture=/Game/Athena/Editor/MapLayersAutomation/Output/Terrain_Minimap_Automation
   )
   "On Selection Changed (RenderSize)"!() -> (OutputDelegate, then, SelectedItem:String, SelectionType:ESelectInfo Enum) {is=title K2Node_ComponentBoundEvent}
   n2 = Select!(
    Option 0:int=2048
    Option 1:int=4096
    Option 2:int=8192
    Index:int=GetSelectedIndex!(self:"Searchable Combo Box (String) Object Reference"=Get RenderSize() -> (RenderSize:"Searchable Combo Box (String) Object Reference")) -> (ReturnValue:int=0)
   ) -> (ReturnValue:int)
   > ResizeRenderTarget2D!(execute, TextureRenderTarget:Texture Render Target 2D Object Reference=Get MapRenderTarget(), Width:int=@n2, Height:int=@n2) -> (then)
   Construct Layers List() {is=title K2Node_CustomEvent}
   > ClearChildren(self=Get LayersContainer() -> (LayersContainer:Editor Utility Scroll Box Object Reference))
   > n3 = Reverse for Each Loop!(Exec:Exec, Array:Array of Struct Map Layer Automation Params Structures=Get Layers()) -> (LoopBody:Exec, ArrayIndex:int, ArrayElement:Struct Map Layer Automation Params Structure, Completed:Exec) {is=title K2Node_MacroInstance}
   > n4 = Create Widget!(
    execute:Exec
    WorldContextObject:Object Reference
    Class:User Widget Class Reference=/Game/Athena/Editor/MapLayersAutomation/SubWidgets/EUW_MapLayerEntry.EUW_MapLayerEntry_C
    OwningPlayer:Player Controller Object Reference
    Parent:EUW Map Layers Automation Object Reference=Self-Reference()
    Index:int=@n3
   ) -> (then:Exec, ReturnValue:EUW Map Layer Entry Object Reference)
   > AddChild(self=Get LayersContainer(), Content=@n4)
   > Render Map()
   Render Map() {is=title K2Node_CustomEvent}
   > RenderMapFunction()
   n5 = LoadLayers() {is=title K2Node_CustomEvent}
   Event PreConstruct()
   > n6 = CreateRenderTarget2D!(
    execute
    WorldContextObject:Object Reference
    Width:int=4096
    Height:int=4096
    Format:ETextureRenderTargetFormat Enum=RTF_RGBA16f
    ClearColor:Linear Color Structure=(R=0,G=0,B=0,A=1)
    bAutoGenerateMipMaps:bool=false
    bSupportUAVs:bool=false
   ) -> (then, ReturnValue:Texture Render Target 2D Object Reference)
   > Set MapRenderTarget(MapRenderTarget:Texture Render Target 2D Object Reference=@n6) -> (Output_Get:Texture Render Target 2D Object Reference)
   > SetTextureParameterValue(
    self=Get Material!(self:EUW Map Layers Viewport Object Reference=Get MapViewport() -> (MapViewport:EUW Map Layers Viewport Object Reference)) -> (Material:Material Instance Dynamic Object Reference)
    ParameterName=Image
    Value=Get MapRenderTarget()
   )
   Event Construct()
   > LoadLayers() {is=title K2Node_CallFunction}
   > Set Timer by Event(Delegate=Create Event(), Time=0.1, bLooping=true, bMaxOncePerFrame=true)
   Event Tick()
   Get Data Table Row!(execute:Exec, DataTable:Data Table Object Reference=Get DataTable() -> (DataTable:Data Table Object Reference), RowName:Name) -> (then:Exec, RowNotFound:Exec, ReturnValue:Wildcard)
  ]}
 Layer Draw Operation {nodes [
   n1 = Layer Draw Operation!() -> (then, Struct_MapLayerAutomationParams:"Struct Map Layer Automation Params Structure (by ref)") {is=title K2Node_FunctionEntry}
   n2 = Break Struct Map Layer Automation Params!(Struct_MapLayerAutomationParams:"Struct Map Layer Automation Params Structure (by ref)"=@n1) -> (
    LayerName_17_#1:String
    Operation_2_#2:Enum_MapBlendLayers Enum
    Texture_19_#3:Texture 2D Object Reference
    UniformColor_12_#4:Linear Color Structure
    AsMask_20_#5:bool=~
    InvertMask_14_#6:bool=~
    Visibility_22_#7:bool=~
   ) {is=title K2Node_BreakStruct}
   > Branch(Condition=@n2.Visibility_22_#7)
   > Is Valid(InputObject=@n2.Texture_19_#3)
   > Set RT_Buffer(RT_Buffer:Texture Render Target 2D Object Reference=Get MapRenderTarget()) -> (Output_Get:Texture Render Target 2D Object Reference)
   > n3 = CreateDynamicMaterialInstance(Parent=/Game/Athena/Editor/MapLayersAutomation/Materials/PP_MapLayerBlendMode, OptionalName=None, CreationFlags=None)
   > n4 = Set LocalOperationMaterial(LocalOperationMaterial:Material Instance Dynamic Object Reference=@n3) -> (Output_Get:Material Instance Dynamic Object Reference)
   > SetTextureParameterValue(self=@n4, ParameterName=Base, Value=Get MapRenderTarget())
   n5 = Get LocalOperationMaterial() -> (LocalOperationMaterial:Material Instance Dynamic Object Reference)
   > SetTextureParameterValue(self=@n5, ParameterName=Layer, Value=@n2.Texture_19_#3)
   > SetVectorParameterValue(self=@n5, ParameterName=Uniform Color, Value=@n2.UniformColor_12_#4)
   > SetScalarParameterValue(
    self=@n5
    ParameterName=Operation Index
    Value="To Float (Byte)"!(InByte:Byte=@n2.Operation_2_#2) -> (ReturnValue:double)
   )
   > SetScalarParameterValue(self=@n5, ParameterName=As Mask, Value="To Float (Boolean)"(InBool=@n2.AsMask_20_#5))
   > SetScalarParameterValue(self=@n5, ParameterName=Invert Mask, Value="To Float (Boolean)"(InBool=@n2.InvertMask_14_#6))
   > DrawMaterialToRenderTarget!(
    execute
    WorldContextObject:Object Reference
    TextureRenderTarget:Texture Render Target 2D Object Reference=Get MapRenderTarget()
    Material:Material Interface Object Reference=Get LocalOperationMaterial()
   ) -> (then)
  ]}
 Load {nodes [
   n1 = Export Data Table to CSV String!(execute, DataTable:Data Table Object Reference=Get DataTable()) -> (then, OutCSVString:String, ReturnValue:bool)
   > PrintString(InString=@n1)
   Load() {is=title K2Node_FunctionEntry}
   > Clear(TargetArray:Array of Struct Map Layer Automation Params Structures=Get Layers())
   > n2 = GetDataTableRowNames(Table=/Game/Athena/Editor/MapLayersAutomation/Misc/Table_MapLayerBatchOperations)
   > n3 = For Each Loop(Array:Array of Names=@n2) -> (Array Element:Name)
   > n4 = Get Data Table Row Table_MapLayerBatchOperations!(
    execute:Exec
    DataTable:Data Table Object Reference=&Table_MapLayerBatchOperations
    RowName:Name=@n3
   ) -> (then:Exec, RowNotFound:Exec, ReturnValue:Struct Map Layer Automation Params Structure) {is=title K2Node_GetDataTableRow}
   > Add(TargetArray=Get Layers(), NewItem=@n4)
   Get DataTable()
  ]}
 Populate Layers Menu {nodes [Populate Layers Menu() {is=title K2Node_FunctionEntry}]}
 RenderMapFunction {nodes [
   RenderMapFunction() {is=title K2Node_FunctionEntry}
   n1 = Get RenderSize()
   > DrawMaterialToRenderTarget(
    TextureRenderTarget=Get MapRenderTarget()
    Material="Get (a copy)"(
     Array:Array of Material Interface Object References=Get CopyMapMaterials() -> (CopyMapMaterials:Array of Material Interface Object References)
     Dimension 1=GetSelectedIndex(self=@n1) -> (ReturnValue=0)
    ) -> (Output:Material Interface Object Reference)
   )
   > n2 = For Each Loop(Array:Array of Struct Map Layer Automation Params Structures=Get Layers()) -> (Array Element:Struct Map Layer Automation Params Structure)
   > Layer Draw Operation!(execute, self, Struct_MapLayerAutomationParams:"Struct Map Layer Automation Params Structure (by ref)"=@n2) -> (then) {is=title K2Node_CallFunction}
  ]}
 Save {nodes [
   n1 = Export Data Table to CSV String(DataTable=&Table_MapLayerBatchOperations)
   > PrintString(InString=@n1)
   Save() {is=title K2Node_FunctionEntry}
   > Set CSV_String(CSV_String:String="---,Layer Name,Operation,Texture,Uniform Color,As Mask,Invert Mask,Visibility\r\n") -> (Output_Get:String)
   > n2 = For Each Loop(Array=Get Layers())
   Fill Data Table from CSV String!(
    execute=@n2.Completed
    DataTable:Data Table Object Reference=Get DataTable()
    CSVString:String=Get CSV_String() -> (CSV_String:String)
    ImportRowStruct:Script Struct Object Reference
   ) -> (then, ReturnValue:bool)
   n3 = Break Struct Map Layer Automation Params(Struct_MapLayerAutomationParams=@n2) -> (
    AsMask_20_#5=~
    InvertMask_14_#6=~
    Visibility_22_#7=~
   )
   n4 = "To String (Integer)"(InInt=@n2.Array Index)
   Set CSV_String(
    execute=@n2.LoopBody
    CSV_String=Append!(
     A=Get CSV_String()
     B=Append!(A=Row_, B=@n4) -> (ReturnValue)
     C:String=",\""
     D:String=@n3
     E:String="\",\""
     F:String=Enum to String!(Enumerator:Enum_MapBlendLayers Enum=@n3.Operation_2_#2) -> (ReturnValue:String)
     G:String="\",\""
     H:String=GetDisplayName(Object=@n3.Texture_19_#3)
     I:String="\",\""
     J:String="To String (LinearColor)"!(InColor:Linear Color Structure=@n3.UniformColor_12_#4) -> (ReturnValue:String)
     K:String="\",\""
     L:String="To String (Boolean)"(InBool=@n3.AsMask_20_#5)
     M:String="\",\""
     N:String="To String (Boolean)"(InBool=@n3.InvertMask_14_#6)
     O:String="\",\""
     P:String="To String (Boolean)"(InBool=@n3.Visibility_22_#7)
     Q:String="\"\r\n"
    ) -> (ReturnValue)
   )
   n5 = For Each Loop(Array=Get Layers())
   > AddDataTableRow!(
    execute
    DataTable:Data Table Object Reference=&Table_MapLayerBatchOperations
    RowName:"Name (by ref)"=String To Name(InString="To String (Integer)"(InInt=@n5.Array Index))
    RowData:Struct Map Layer Automation Params Structure=@n5
   ) -> (then)
   Fill Data Table from JSON String!(execute, DataTable:Data Table Object Reference=Get DataTable(), JSONString:String="[]", ImportRowStruct:Script Struct Object Reference) -> (then, ReturnValue:bool)
  ]}
]
variables [
 CopyMapMaterials {
  type="array<MaterialInterface>"
  default [
   /Game/Athena/Editor/CopyMinimap/M_CopyMinimap_2k
   /Game/Athena/Editor/CopyMinimap/M_CopyMinimap
   /Game/Athena/Editor/CopyMinimap/M_CopyMinimap_8k
   /Game/Athena/Editor/CopyMinimap/M_CopyMinimap_Undiscovered2k
   /Game/Athena/Editor/CopyMinimap/M_CopyMinimap_Undiscovered4k
   /Game/Athena/Editor/CopyMinimap/M_CopyMinimap_Undiscovered8k
  ]
 }
 DataTable {type=DataTable, default=&Table_MapLayerBatchOperations}
 Layers {type="array<Struct_MapLayerAutomationParams>", default []}
 LocalOperationMaterial {type=MaterialInstanceDynamic}
 MapRenderTarget {type=TextureRenderTarget2D, default=null}
 MousePositionBuffered {type=Vector2D, default=<0 0>}
 ViewportImagePositionBuffered {type=Vector2D, default=<0 0>}
]
widgets {
 info {rootWidgetClass=&CanvasPanel, widgetCount=16}
 widgets [
  CanvasPanel_37 {widgetClassPath=&CanvasPanel, widget=CanvasPanel_37}
  Container {
   widgetClassPath=&HorizontalBox
   parent=CanvasPanel_37
   widget=Container
   slot {class=CanvasPanelSlot, properties {layoutData {alignment=<0 0>, anchors {maximum=<1 1>, minimum=<0 0>}, offsets=<ltrb 0 0 0 0>}}}
  }
  Details {
   widgetClassPath=&CanvasPanel
   bIsVariable=true
   parent=Container
   widget=Details
   slot {class=HorizontalBoxSlot, properties {size {sizeRule=Fill, value=0.2}}}
  }
  RenderSize {
   widgetClassPath=/Script/FortniteEditor.SearchableStringComboBox
   bIsVariable=true
   parent=Details
   widget=RenderSize
   properties {
    selectedOption=4096
    defaultOptions [2048,4096,8192]
    foregroundColor {colorUseRule=UseColor_Specified, specifiedColor=<rgba 0.5271 0.5271 0.5271 1>}
    itemStyle {
     bUseParentRowBrush=false
     activeBrush {
      bIsDynamicallyLoaded=false
      drawAs=Image
      imageType=NoImage
      mirroring=NoMirror
      resourceName=None
      resourceObject=null
      tiling=NoTile
      imageSize=<0 0>
      margin=<ltrb 0 0 0 0>
      outlineSettings {
       bUseBrushTransparency=false
       roundingType=HalfHeightRadius
       width=0
       color {colorUseRule=UseColor_Specified, specifiedColor=<rgba 0 0 0 0>}
       cornerRadii=<0 0 0 0>
      }
      tintColor {colorUseRule=UseColor_Specified, specifiedColor=<rgba 0 0.162 0.7454 1>}
      uVRegion {bIsValid=false, max=<0 0>, min=<0 0>}
     }
     activeHighlightedBrush {
      bIsDynamicallyLoaded=false
      drawAs=Image
      imageType=NoImage
      mirroring=NoMirror
      resourceName=None
      resourceObject=null
      tiling=NoTile
      imageSize=<0 0>
      margin=<ltrb 0 0 0 0>
      outlineSettings {
       bUseBrushTransparency=false
       roundingType=HalfHeightRadius
       width=0
       color {colorUseRule=UseColor_Specified, specifiedColor=<rgba 0 0 0 0>}
       cornerRadii=<0 0 0 0>
      }
      tintColor {colorUseRule=UseColor_Specified, specifiedColor=<rgba 0.0252 0.0319 0.0423 1>}
      uVRegion {bIsValid=false, max=<0 0>, min=<0 0>}
     }
     activeHoveredBrush {
      bIsDynamicallyLoaded=false
      drawAs=Image
      imageType=NoImage
      mirroring=NoMirror
      resourceName=None
      resourceObject=null
      tiling=NoTile
      imageSize=<0 0>
      margin=<ltrb 0 0 0 0>
      outlineSettings {
       bUseBrushTransparency=false
       roundingType=HalfHeightRadius
       width=0
       color {colorUseRule=UseColor_Specified, specifiedColor=<rgba 0 0 0 0>}
       cornerRadii=<0 0 0 0>
      }
      tintColor {colorUseRule=UseColor_Specified, specifiedColor=<rgba 0 0.162 0.7454 1>}
      uVRegion {bIsValid=false, max=<0 0>, min=<0 0>}
     }
     dropIndicator_Above {
      bIsDynamicallyLoaded=false
      drawAs=Box
      imageType=FullColor
      mirroring=NoMirror
      resourceName=../../../Engine/Content/Slate/Common/DropZoneIndicator_Above.png
      resourceObject=null
      tiling=NoTile
      imageSize=<0 0>
      margin=<ltrb 0.625 0.625 0 0>
      outlineSettings {
       bUseBrushTransparency=false
       roundingType=HalfHeightRadius
       width=0
       color {colorUseRule=UseColor_Specified, specifiedColor=<rgba 0 0 0 0>}
       cornerRadii=<0 0 0 0>
      }
      tintColor {colorUseRule=UseColor_Specified, specifiedColor=<rgba 0.0091 0.3515 0.9647 1>}
      uVRegion {bIsValid=false, max=<0 0>, min=<0 0>}
     }
     dropIndicator_Below {
      bIsDynamicallyLoaded=false
      drawAs=Box
      imageType=FullColor
      mirroring=NoMirror
      resourceName=../../../Engine/Content/Slate/Common/DropZoneIndicator_Below.png
      resourceObject=null
      tiling=NoTile
      imageSize=<0 0>
      margin=<ltrb 0.625 0 0 0.625>
      outlineSettings {
       bUseBrushTransparency=false
       roundingType=HalfHeightRadius
       width=0
       color {colorUseRule=UseColor_Specified, specifiedColor=<rgba 0 0 0 0>}
       cornerRadii=<0 0 0 0>
      }
      tintColor {colorUseRule=UseColor_Specified, specifiedColor=<rgba 0.0091 0.3515 0.9647 1>}
      uVRegion {bIsValid=false, max=<0 0>, min=<0 0>}
     }
     dropIndicator_Onto {
      bIsDynamicallyLoaded=false
      drawAs=Box
      imageType=FullColor
      mirroring=NoMirror
      resourceName=../../../Engine/Content/Slate/Common/DropZoneIndicator_Onto.png
      resourceObject=null
      tiling=NoTile
      imageSize=<0 0>
      margin=<ltrb 0.25 0.25 0.25 0.25>
      outlineSettings {
       bUseBrushTransparency=false
       roundingType=HalfHeightRadius
       width=0
       color {colorUseRule=UseColor_Specified, specifiedColor=<rgba 0 0 0 0>}
       cornerRadii=<0 0 0 0>
      }
      tintColor {colorUseRule=UseColor_Specified, specifiedColor=<rgba 0.0091 0.3515 0.9647 1>}
      uVRegion {bIsValid=false, max=<0 0>, min=<0 0>}
     }
     evenRowBackgroundBrush {
      bIsDynamicallyLoaded=false
      drawAs=Image
      imageType=NoImage
      mirroring=NoMirror
      resourceName=None
      resourceObject=null
      tiling=NoTile
      imageSize=<0 0>
      margin=<ltrb 0 0 0 0>
      outlineSettings {
       bUseBrushTransparency=false
       roundingType=HalfHeightRadius
       width=0
       color {colorUseRule=UseColor_Specified, specifiedColor=<rgba 0 0 0 0>}
       cornerRadii=<0 0 0 0>
      }
      tintColor {colorUseRule=UseColor_Specified, specifiedColor=<rgba 0.0103 0.0103 0.0103 1>}
      uVRegion {bIsValid=false, max=<0 0>, min=<0 0>}
     }
     evenRowBackgroundHoveredBrush {
      bIsDynamicallyLoaded=false
      drawAs=Image
      imageType=NoImage
      mirroring=NoMirror
      resourceName=None
      resourceObject=null
      tiling=NoTile
      imageSize=<0 0>
      margin=<ltrb 0 0 0 0>
      outlineSettings {
       bUseBrushTransparency=false
       roundingType=HalfHeightRadius
       width=0
       color {colorUseRule=UseColor_Specified, specifiedColor=<rgba 0 0 0 0>}
       cornerRadii=<0 0 0 0>
      }
      tintColor {colorUseRule=UseColor_Specified, specifiedColor=<rgba 0.0176 0.0176 0.0176 1>}
      uVRegion {bIsValid=false, max=<0 0>, min=<0 0>}
     }
     inactiveBrush {
      bIsDynamicallyLoaded=false
      drawAs=Image
      imageType=NoImage
      mirroring=NoMirror
      resourceName=None
      resourceObject=null
      tiling=NoTile
      imageSize=<0 0>
      margin=<ltrb 0 0 0 0>
      outlineSettings {
       bUseBrushTransparency=false
       roundingType=HalfHeightRadius
       width=0
       color {colorUseRule=UseColor_Specified, specifiedColor=<rgba 0 0 0 0>}
       cornerRadii=<0 0 0 0>
      }
      tintColor {colorUseRule=UseColor_Specified, specifiedColor=<rgba 0.0513 0.0953 0.159 1>}
      uVRegion {bIsValid=false, max=<0 0>, min=<0 0>}
     }
     inactiveHighlightedBrush {
      bIsDynamicallyLoaded=false
      drawAs=Image
      imageType=NoImage
      mirroring=NoMirror
      resourceName=None
      resourceObject=null
      tiling=NoTile
      imageSize=<0 0>
      margin=<ltrb 0 0 0 0>
      outlineSettings {
       bUseBrushTransparency=false
       roundingType=HalfHeightRadius
       width=0
       color {colorUseRule=UseColor_Specified, specifiedColor=<rgba 0 0 0 0>}
       cornerRadii=<0 0 0 0>
      }
      tintColor {colorUseRule=UseColor_Specified, specifiedColor=<rgba 0.0252 0.0319 0.0423 1>}
      uVRegion {bIsValid=false, max=<0 0>, min=<0 0>}
     }
     inactiveHoveredBrush {
      bIsDynamicallyLoaded=false
      drawAs=Image
      imageType=NoImage
      mirroring=NoMirror
      resourceName=None
      resourceObject=null
      tiling=NoTile
      imageSize=<0 0>
      margin=<ltrb 0 0 0 0>
      outlineSettings {
       bUseBrushTransparency=false
       roundingType=HalfHeightRadius
       width=0
       color {colorUseRule=UseColor_Specified, specifiedColor=<rgba 0 0 0 0>}
       cornerRadii=<0 0 0 0>
      }
      tintColor {colorUseRule=UseColor_Specified, specifiedColor=<rgba 0.0513 0.0953 0.159 1>}
      uVRegion {bIsValid=false, max=<0 0>, min=<0 0>}
     }
     oddRowBackgroundBrush {
      bIsDynamicallyLoaded=false
      drawAs=Image
      imageType=NoImage
      mirroring=NoMirror
      resourceName=None
      resourceObject=null
      tiling=NoTile
      imageSize=<0 0>
      margin=<ltrb 0 0 0 0>
      outlineSettings {
       bUseBrushTransparency=false
       roundingType=HalfHeightRadius
       width=0
       color {colorUseRule=UseColor_Specified, specifiedColor=<rgba 0 0 0 0>}
       cornerRadii=<0 0 0 0>
      }
      tintColor {colorUseRule=UseColor_Specified, specifiedColor=<rgba 0.0103 0.0103 0.0103 1>}
      uVRegion {bIsValid=false, max=<0 0>, min=<0 0>}
     }
     oddRowBackgroundHoveredBrush {
      bIsDynamicallyLoaded=false
      drawAs=Image
      imageType=NoImage
      mirroring=NoMirror
      resourceName=None
      resourceObject=null
      tiling=NoTile
      imageSize=<0 0>
      margin=<ltrb 0 0 0 0>
      outlineSettings {
       bUseBrushTransparency=false
       roundingType=HalfHeightRadius
       width=0
       color {colorUseRule=UseColor_Specified, specifiedColor=<rgba 0 0 0 0>}
       cornerRadii=<0 0 0 0>
      }
      tintColor {colorUseRule=UseColor_Specified, specifiedColor=<rgba 0.0176 0.0176 0.0176 1>}
      uVRegion {bIsValid=false, max=<0 0>, min=<0 0>}
     }
     parentRowBackgroundBrush {
      bIsDynamicallyLoaded=false
      drawAs=Image
      imageType=NoImage
      mirroring=NoMirror
      resourceName=None
      resourceObject=null
      tiling=NoTile
      imageSize=<32 32>
      margin=<ltrb 0 0 0 0>
      outlineSettings {
       bUseBrushTransparency=false
       roundingType=HalfHeightRadius
       width=0
       color {colorUseRule=UseColor_Specified, specifiedColor=<rgba 0 0 0 0>}
       cornerRadii=<0 0 0 0>
      }
      tintColor {colorUseRule=UseColor_Specified, specifiedColor=<rgba 1 1 1 1>}
      uVRegion {bIsValid=false, max=<0 0>, min=<0 0>}
     }
     parentRowBackgroundHoveredBrush {
      bIsDynamicallyLoaded=false
      drawAs=Image
      imageType=NoImage
      mirroring=NoMirror
      resourceName=None
      resourceObject=null
      tiling=NoTile
      imageSize=<32 32>
      margin=<ltrb 0 0 0 0>
      outlineSettings {
       bUseBrushTransparency=false
       roundingType=HalfHeightRadius
       width=0
       color {colorUseRule=UseColor_Specified, specifiedColor=<rgba 0 0 0 0>}
       cornerRadii=<0 0 0 0>
      }
      tintColor {colorUseRule=UseColor_Specified, specifiedColor=<rgba 1 1 1 1>}
      uVRegion {bIsValid=false, max=<0 0>, min=<0 0>}
     }
     selectedTextColor {colorUseRule=UseColor_Specified, specifiedColor=<rgba 1 1 1 1>}
     selectorFocusedBrush {
      bIsDynamicallyLoaded=false
      drawAs=Border
      imageType=FullColor
      mirroring=NoMirror
      resourceName=../../../Engine/Content/Slate/Common/Selector.png
      resourceObject=null
      tiling=Both
      imageSize=<0 0>
      margin=<ltrb 0.25 0.25 0.25 0.25>
      outlineSettings {
       bUseBrushTransparency=false
       roundingType=HalfHeightRadius
       width=0
       color {colorUseRule=UseColor_Specified, specifiedColor=<rgba 0 0 0 0>}
       cornerRadii=<0 0 0 0>
      }
      tintColor {colorUseRule=UseColor_Specified, specifiedColor=<rgba 0.701 0.225 0.003 1>}
      uVRegion {bIsValid=false, max=<0 0>, min=<0 0>}
     }
     textColor {colorUseRule=UseColor_Specified, specifiedColor=<rgba 0.5271 0.5271 0.5271 1>}
    }
    widgetStyle {
     comboButtonStyle {
      downArrowAlign=VAlign_Center
      buttonStyle {
       clickedSlateSound {resourceObject=null}
       disabled {
        bIsDynamicallyLoaded=false
        drawAs=RoundedBox
        imageType=NoImage
        mirroring=NoMirror
        resourceName=None
        resourceObject=null
        tiling=NoTile
        imageSize=<0 0>
        margin=<ltrb 0 0 0 0>
        outlineSettings {
         bUseBrushTransparency=false
         roundingType=FixedRadius
         width=1
         color {colorUseRule=UseColor_ColorTable, specifiedColor=<rgba 0 0 0 0>}
         cornerRadii=<4 4 4 4>
        }
        tintColor {colorUseRule=UseColor_Specified, specifiedColor=<rgba 0.0395 0.0395 0.0395 1>}
        uVRegion {bIsValid=false, max=<0 0>, min=<0 0>}
       }
       disabledForeground {colorUseRule=UseColor_ColorTable, specifiedColor=<rgba 0 0 0 0>}
       hovered {
        bIsDynamicallyLoaded=false
        drawAs=RoundedBox
        imageType=NoImage
        mirroring=NoMirror
        resourceName=None
        resourceObject=null
        tiling=NoTile
        imageSize=<0 0>
        margin=<ltrb 0 0 0 0>
        outlineSettings {
         bUseBrushTransparency=false
         roundingType=FixedRadius
         width=1
         color {colorUseRule=UseColor_ColorTable, specifiedColor=<rgba 0 0 0 0>}
         cornerRadii=<4 4 4 4>
        }
        tintColor {colorUseRule=UseColor_Specified, specifiedColor=<rgba 0.0048 0.0048 0.0048 1>}
        uVRegion {bIsValid=false, max=<0 0>, min=<0 0>}
       }
       hoveredForeground {colorUseRule=UseColor_ColorTable, specifiedColor=<rgba 0 0 0 0>}
       hoveredSlateSound {resourceObject=null}
       normal {
        bIsDynamicallyLoaded=false
        drawAs=RoundedBox
        imageType=NoImage
        mirroring=NoMirror
        resourceName=None
        resourceObject=null
        tiling=NoTile
        imageSize=<0 0>
        margin=<ltrb 0 0 0 0>
        outlineSettings {
         bUseBrushTransparency=false
         roundingType=FixedRadius
         width=1
         color {colorUseRule=UseColor_ColorTable, specifiedColor=<rgba 0 0 0 0>}
         cornerRadii=<4 4 4 4>
        }
        tintColor {colorUseRule=UseColor_Specified, specifiedColor=<rgba 0.0048 0.0048 0.0048 1>}
        uVRegion {bIsValid=false, max=<0 0>, min=<0 0>}
       }
       normalForeground {colorUseRule=UseColor_ColorTable, specifiedColor=<rgba 0 0 0 0>}
       normalPadding=<ltrb 8 1 8 1>
       pressed {
        bIsDynamicallyLoaded=false
        drawAs=RoundedBox
        imageType=NoImage
        mirroring=NoMirror
        resourceName=None
        resourceObject=null
        tiling=NoTile
        imageSize=<0 0>
        margin=<ltrb 0 0 0 0>
        outlineSettings {
         bUseBrushTransparency=false
         roundingType=FixedRadius
         width=1
         color {colorUseRule=UseColor_ColorTable, specifiedColor=<rgba 0 0 0 0>}
         cornerRadii=<4 4 4 4>
        }
        tintColor {colorUseRule=UseColor_Specified, specifiedColor=<rgba 0.0048 0.0048 0.0048 1>}
        uVRegion {bIsValid=false, max=<0 0>, min=<0 0>}
       }
       pressedForeground {colorUseRule=UseColor_ColorTable, specifiedColor=<rgba 0 0 0 0>}
       pressedPadding=<ltrb 8 2 8 0>
       pressedSlateSound {resourceObject=null}
      }
      contentPadding=<ltrb 0 0 0 0>
      downArrowImage {
       bIsDynamicallyLoaded=false
       drawAs=Image
       imageType=Vector
       mirroring=NoMirror
       resourceName=../../../Engine/Content/Slate/Starship/CoreWidgets/ComboBox/wide-chevron-down.svg
       resourceObject=null
       tiling=NoTile
       imageSize=<16 16>
       margin=<ltrb 0 0 0 0>
       outlineSettings {
        bUseBrushTransparency=false
        roundingType=HalfHeightRadius
        width=0
        color {colorUseRule=UseColor_Specified, specifiedColor=<rgba 0 0 0 0>}
        cornerRadii=<0 0 0 0>
       }
       tintColor {colorUseRule=UseColor_Specified, specifiedColor=<rgba 1 1 1 1>}
       uVRegion {bIsValid=false, max=<0 0>, min=<0 0>}
      }
      downArrowPadding=<ltrb 2 2 2 2>
      menuBorderBrush {
       bIsDynamicallyLoaded=false
       drawAs=RoundedBox
       imageType=NoImage
       mirroring=NoMirror
       resourceName=None
       resourceObject=null
       tiling=NoTile
       imageSize=<0 0>
       margin=<ltrb 0 0 0 0>
       outlineSettings {
        bUseBrushTransparency=false
        roundingType=FixedRadius
        width=1
        color {colorUseRule=UseColor_ColorTable, specifiedColor=<rgba 0 0 0 0>}
        cornerRadii=<0 0 0 0>
       }
       tintColor {colorUseRule=UseColor_Specified, specifiedColor=<rgba 0.0395 0.0395 0.0395 1>}
       uVRegion {bIsValid=false, max=<0 0>, min=<0 0>}
      }
      menuBorderPadding=<ltrb 0 0 0 0>
      shadowColorAndOpacity=<rgba 0 0 0 1>
      shadowOffset=<0 0>
     }
     contentPadding=<ltrb 0 0 0 0>
     menuRowPadding=<ltrb 8 3 8 3>
     pressedSlateSound {resourceObject=null}
     selectionChangeSlateSound {resourceObject=null}
    }
   }
   slot {class=CanvasPanelSlot, properties {layoutData {alignment=<0 0>, anchors {maximum=<0 0>, minimum=<0 0>}, offsets=<ltrb 244 40 100 40>}}}
  }
  LoadBtn {
   widgetClassPath=&EditorUtilityButton
   parent=Details
   widget=LoadBtn
   slot {class=CanvasPanelSlot, properties {layoutData {alignment=<0 0>, anchors {maximum=<0 0>, minimum=<0 0>}, offsets=<ltrb 80 108 100 30>}}}
  }
  TextBlock {
   widgetClassPath=&TextBlock
   parent=LoadBtn
   widget=TextBlock
   properties {
    justification=Center
    text=Load
    colorAndOpacity {colorUseRule=UseColor_Specified, specifiedColor=<rgba 0.7188 0.7188 0.7188 1>}
    font {
     bForceMonospaced=false
     bMaterialIsStencil=false
     fontMaterial=null
     fontObject=&Roboto
     letterSpacing=0
     monospacedWidth=1
     size=8.25
     skewAmount=0
     typefaceFontName=Bold
     outlineSettings {
      bApplyOutlineToDropShadows=false
      bMiteredCorners=false
      bSeparateFillAlpha=false
      outlineMaterial=null
      outlineSize=0
      outlineColor=<rgba 0 0 0 1>
     }
    }
   }
   slot {class=ButtonSlot}
  }
  SaveBtn {
   widgetClassPath=&EditorUtilityButton
   parent=Details
   widget=SaveBtn
   slot {class=CanvasPanelSlot, properties {layoutData {alignment=<0 0>, anchors {maximum=<0 0>, minimum=<0 0>}, offsets=<ltrb 224 108 100 30>}}}
  }
  TextBlock_1 {
   widgetClassPath=&TextBlock
   parent=SaveBtn
   widget=TextBlock_1
   properties {
    justification=Center
    text=Save
    colorAndOpacity {colorUseRule=UseColor_Specified, specifiedColor=<rgba 0.7188 0.7188 0.7188 1>}
    font {
     bForceMonospaced=false
     bMaterialIsStencil=false
     fontMaterial=null
     fontObject=&Roboto
     letterSpacing=0
     monospacedWidth=1
     size=8.25
     skewAmount=0
     typefaceFontName=Bold
     outlineSettings {
      bApplyOutlineToDropShadows=false
      bMiteredCorners=false
      bSeparateFillAlpha=false
      outlineMaterial=null
      outlineSize=0
      outlineColor=<rgba 0 0 0 1>
     }
    }
   }
   slot {class=ButtonSlot}
  }
  EUW_ResizeBox {
   widgetClassPath=/Game/Athena/Editor/MapLayersAutomation/SubWidgets/EUW_ResizeDivider
   bIsVariable=true
   parent=Container
   widget=EUW_ResizeBox
   slot {class=HorizontalBoxSlot, properties {horizontalAlignment=HAlign_Center}}
  }
  Viewport {
   widgetClassPath=&CanvasPanel
   bIsVariable=true
   parent=Container
   widget=Viewport
   slot {class=HorizontalBoxSlot, properties {size {sizeRule=Fill, value=0.55}}}
  }
  MapViewport {
   widgetClassPath=/Game/Athena/Editor/MapLayersAutomation/SubWidgets/EUW_MapLayersViewport
   bIsVariable=true
   parent=Viewport
   widget=MapViewport
   slot {class=CanvasPanelSlot, properties {layoutData {alignment=<0 0>, anchors {maximum=<1 1>, minimum=<0 0>}, offsets=<ltrb 0 0 0 0>}}}
  }
  EUW_ResizeBox_1 {
   widgetClassPath=&EUW_ResizeDivider
   bIsVariable=true
   parent=Container
   widget=EUW_ResizeBox_1
   slot {class=HorizontalBoxSlot, properties {horizontalAlignment=HAlign_Center}}
  }
  LayersCanvas {
   widgetClassPath=&CanvasPanel
   bIsVariable=true
   parent=Container
   widget=LayersCanvas
   slot {class=HorizontalBoxSlot, properties {size {sizeRule=Fill, value=0.25}}}
  }
  AddLayerBtn {
   widgetClassPath=&EditorUtilityButton
   parent=LayersCanvas
   widget=AddLayerBtn
   slot {class=CanvasPanelSlot, properties {layoutData {alignment=<0 0>, anchors {maximum=<0 0>, minimum=<0 0>}, offsets=<ltrb 40 40 100 30>}}}
  }
  TextBlock_310 {
   widgetClassPath=&TextBlock
   parent=AddLayerBtn
   widget=TextBlock_310
   properties {
    justification=Center
    text=New Layer
    colorAndOpacity {colorUseRule=UseColor_Specified, specifiedColor=<rgba 0.7188 0.7188 0.7188 1>}
    font {
     bForceMonospaced=false
     bMaterialIsStencil=false
     fontMaterial=null
     fontObject=&Roboto
     letterSpacing=0
     monospacedWidth=1
     size=8.25
     skewAmount=0
     typefaceFontName=Bold
     outlineSettings {
      bApplyOutlineToDropShadows=false
      bMiteredCorners=false
      bSeparateFillAlpha=false
      outlineMaterial=null
      outlineSize=0
      outlineColor=<rgba 0 0 0 1>
     }
    }
   }
   slot {class=ButtonSlot}
  }
  LayersContainer {
   widgetClassPath=&EditorUtilityScrollBox
   bIsVariable=true
   parent=LayersCanvas
   widget=LayersContainer
   slot {class=CanvasPanelSlot, properties {layoutData {alignment=<0 0>, anchors {maximum=<1 1>, minimum=<0 0.1>}, offsets=<ltrb 0 0 0 0>}}}
  }
 ]
}
