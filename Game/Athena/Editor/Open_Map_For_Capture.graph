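# Editor utility task that opens a map for minimap capture by delegating to the Capture_Minimap task.
# On BeginExecution it constructs a Capture_Minimap object with LoadMapOnly and LoadEverything enabled, passing in the Map soft object path.
# It registers and executes that object as a child task of itself, then finishes when the child reports completion.
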EditorUtilityBlueprint /Game/Athena/Editor/Open_Map_For_Capture
graphs [EventGraph {nodes [
   Event BeginExecution()
   > n1 = Construct Object from Class!(
    execute
    Class=&/Capture_Minimap.Capture_Minimap_C
    self=Self-Reference()
    LoadEverything:bool=True
    PreserveExistingState:bool=False
    LoadMapOnly:bool=true
    Resolution:int=0
    Map:Soft Object Path Structure=Get Map() -> (Map:Soft Object Path Structure)
    ImportTargets:Array of Texture 2D Object References
    Scene Capture Class:Scene Capture 2D Class Reference
   ) -> (then, ReturnValue:Capture Minimap Object Reference)
   > n2 = RegisterAndExecuteTask!(execute:Exec, Task:Editor Utility Task Object Reference=@n1, OptionalParentTask:Editor Utility Task Object Reference=Self-Reference()) -> (then:Exec, OnFinished:Exec, Task:Editor Utility Task Object Reference) {is=title K2Node_AsyncAction}
   FinishExecutingTask(execute=@n2.OnFinished)
  ]}]
variables [Map {type=SoftObjectPath, default=""}]
