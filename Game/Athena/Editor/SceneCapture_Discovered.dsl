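; Scene capture 2D actor Blueprint in the Athena editor folder, named for a "Discovered" capture, with configurable show flags and post-process material.
; Its construction script adds the material, when valid, as a full-weight blendable on its PostProcess component.
; On BeginPlay it applies the New Show Flags variable to CaptureComponent2D, so the flag override is set at runtime rather than during construction.
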(event EventBeginPlay
  (SceneCapture|SetSceneCaptureShowFlags (Variables|DecalActor|GetCaptureComponent2D) (Variables|Default|GetNew%20Show%20Flags)))

(fn ConstructionScript
  (Utilities|IsValid (Variables|Default|GetPostProcessMaterial)
    (:Is%20Valid
      (Rendering|AddorUpdateBlendable (Variables|Default|GetPostProcess) (Variables|Default|GetPostProcessMaterial) 1.0 :!class /Script/Engine.PostProcessComponent))))
