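# ShutdownEditor is an Editor Utility Blueprint that closes the Unreal Editor when run as a task.
# On BeginExecution it calls ExitEditor and then FinishExecutingTask to mark the task complete.
# It has no variables or other logic.
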EditorUtilityBlueprint /Game/Athena/Editor/ShutdownEditor
graphs [EventGraph {nodes [
   Event BeginExecution()
   > ExitEditor!(execute) -> (then)
   > FinishExecutingTask()
  ]}]
variables []
