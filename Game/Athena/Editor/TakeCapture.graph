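# Editor utility task that renders the Athena minimap textures, discovered at 2K, 4K and 8K and undiscovered at 2K and 4K.
# After fixing time of day and freezing game time, each pass runs an async warm-up scene capture from MinimapCaptureCamera in the game world.
# Render targets are converted into target Texture2Ds and saved during play.
# Map-loading, console-variable and actor-cleanup helpers exist, but the event graph never calls them.
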EditorUtilityBlueprint /Game/Athena/Editor/TakeCapture
graphs [
 CaptureShotMacro {nodes [
   n1 = Inputs!() -> (execute, Targets:Array of Texture 2D Object References, Resolution:int, SceneCapture:Scene Capture 2D Class Reference)
   > Set Progress!(execute, self, Text:Text="LOCTEXT(\"Begin Capture\")") -> (then) {is=title K2Node_CallFunction}
   > Create Capture!(
    execute:Exec
    Textures:Array of Texture 2D Object References=@n1
    Resolution:int=@n1.Resolution
    WarmUpFrames:int=Get WarmUpFrames() -> (WarmUpFrames:int=0)
    SceneCaptureClass:Scene Capture 2D Class Reference=@n1.SceneCapture
   ) -> (Complete:Exec) {is=title K2Node_MacroInstance}
   > Set Progress(Text="LOCTEXT(\"Capture Complete\")")
   > Outputs!(then:Exec)
  ]}
 CreateCapture {nodes [
   n1 = Inputs!() -> (execute, Textures:Array of Texture 2D Object References, Resolution, WarmUpFrames:int, SceneCaptureClass:Scene Capture 2D Class Reference)
   > n2 = GetGameWorld!(execute, self:Unreal Editor Subsystem Object Reference=Get UnrealEditorSubsystem()) -> (then, ReturnValue:World Object Reference)
   > n3 = FindActorsContainingName!(
    execute
    WorldContextObject:Object Reference=@n2
    ActorClass:Actor Class Reference=/Script/Engine.CameraActor
    Name:String=MinimapCaptureCamera
   ) -> (then, ReturnValue:Array of Actor Object References)
   > n4 = Cast To CameraActor!(execute, Object="Get (a copy)"(Array:Array of Actor Object References=@n3) -> (Output:Actor Object Reference)) -> (then, CastFailed, AsCamera Actor:Camera Actor Object Reference) {is=title K2Node_DynamicCast}
   > Set Progress(Text="LOCTEXT(\"Capturing Shot\")")
   > n5 = CaptureSceneWithWarmupAsync!(
    execute:Exec
    ViewCamera:Camera Component Object Reference=Get CameraComponent!(self:Camera Actor Object Reference=@n4) -> (CameraComponent:Camera Component Object Reference)
    SceneCaptureClass:Scene Capture 2D Class Reference=@n1.SceneCaptureClass
    ResX:int=@n1.Resolution
    ResY:int=@n1.Resolution
    WarmUpFrames:int=@n1.WarmUpFrames
   ) -> (then:Exec, Complete:Exec, Texture:Texture Render Target 2D Object Reference) {is=title K2Node_AsyncAction}
   n6 = For Each Loop(Exec=@n5.Complete, Array:Array of Texture 2D Object References=@n1) -> (Array Element:Texture 2D Object Reference)
   > n7 = Is Valid(InputObject=@n6)
   > ConvertRenderTargetToTexture2DEditorOnly(RenderTarget=@n5, Texture=@n6)
   n8 = ReleaseRenderTarget2D!(execute=[@n9.then, @n7.Is Not Valid], TextureRenderTarget:Texture Render Target 2D Object Reference=@n5) -> (then)
   Outputs!(Complete:Exec=[@n8.then, @n10.then])
   n9 = SaveAssetsWhilePlaying!(execute=@n6.Completed, AssetsToSave:Array of Object References=@n1, bOnlyIfIsDirty:bool=false) -> (then, ReturnValue:bool)
   n10 = Set Progress(execute=@n4.CastFailed, Text="LOCTEXT(\"ERROR: Capture Failed (Missing \\'MinimapCaptureCamera\\')\")")
  ]}
 Delete Actors Of Class {nodes [
   Delete Actors Of Class!() -> (then, ActorClass:Actor Class Reference) {is=title K2Node_FunctionEntry}
   > n1 = GetGameWorld(self=Get UnrealEditorSubsystem())
   > n2 = GetAllActorsOfClass!(execute, WorldContextObject:Object Reference=@n1, ActorClass:Actor Class Reference=/Game/Gadgets/Zipline/Meshes/ZipLine_Pole.ZipLine_Pole_C) -> (then, OutActors:Array of Zip Line Pole Object References)
   > n3 = For Each Loop(Array:Array of Zip Line Pole Object References=@n2) -> (Array Element:Zip Line Pole Object Reference)
   > Hide And Delete Actor!(execute, self, Actor:Actor Object Reference=@n3) -> (then)
  ]}
 Delete All HLODS {nodes [
   Delete All HLODS() {is=title K2Node_FunctionEntry}
   > n1 = GetGameWorld(self=Get UnrealEditorSubsystem())
   > n2 = GetAllActorsOfClass(WorldContextObject=@n1, ActorClass=/Script/FortniteGame.FortHLODSMActor) -> (OutActors:Array of Fort HLODSMActor Object References)
   > n3 = For Each Loop(Array:Array of Fort HLODSMActor Object References=@n2) -> (Array Element:Fort HLODSMActor Object Reference)
   > Hide And Delete Actor(Actor=@n3)
  ]}
 Delete Game Actors Containing Name {nodes [
   n1 = Delete Game Actors Containing Name!() -> (then, Containing Name:String) {is=title K2Node_FunctionEntry}
   > n2 = GetGameWorld(self=Get UnrealEditorSubsystem())
   > n3 = FindActorsContainingName(WorldContextObject=@n2, ActorClass=&Actor, Name=@n1)
   > n4 = For Each Loop(Array:Array of Actor Object References=@n3) -> (Array Element:Actor Object Reference)
   > Hide And Delete Actor(Actor=@n4)
  ]}
 Delete Game Specific Actors {nodes [
   Delete Game Specific Actors() {is=title K2Node_FunctionEntry}
   > n1 = Sequence!(execute) -> (then_0, then_1, then_2:Exec)
   > n2 = GetGameWorld(self=Get UnrealEditorSubsystem())
   > n3 = FindAllComponents!(execute, WorldContextObject:Object Reference=@n2, ComponentClass:Actor Component Class Reference=/Script/FortniteGame.MinimapHiddenComponent) -> (then, ReturnValue:Array of Actor Component Object References)
   > n4 = For Each Loop(Array:Array of Actor Component Object References=@n3) -> (Array Element:Actor Component Object Reference)
   > n5 = Cast To MinimapHiddenComponent!(execute, Object=@n4) -> (then, CastFailed, AsMinimap Hidden Component:Minimap Hidden Component Object Reference) {is=title K2Node_DynamicCast}
   > PrepareForCapture(self:Minimap Hidden Component Object Reference=@n5)
   Delete Actors Of Class!(execute=@n1.then_1, self, ActorClass:Actor Class Reference=&ZipLine_Pole_C) -> (then) {is=title K2Node_CallFunction}
   > Delete Actors Of Class(ActorClass=/Game/Athena/Items/Consumables/ZipLine/SplineZipline/BP_Athena_Environmental_ZipLine_Spline.BP_Athena_Environmental_ZipLine_Spline_C)
   > Delete Actors Of Class(ActorClass=/Game/Building/ActorBlueprints/Containers/Tiered_Chest_Athena.Tiered_Chest_Athena_C)
   > Delete Game Actors Containing Name!(execute, self, Containing Name:String=VolcanoCloud) -> (then) {is=title K2Node_CallFunction}
   > Delete Game Actors Containing Name(Containing Name=BP_GiantCloudSheet)
   > Delete Game Actors Containing Name(Containing Name=B_HidingProp_WilliePete_Large)
   > Delete Game Actors Containing Name(Containing Name=ShadowProxy)
   Delete Game Actors Containing Name(execute=@n1.then_2, Containing Name=Zipline)
  ]}
 EventGraph {nodes [
   Event BeginExecution()
   > AsyncEditorDelay!(execute:Exec, Seconds:float=0.2, MinimumFrames:int=30) -> (then:Exec, Complete:Exec) {is=title K2Node_AsyncAction}
   > Set Progress(Text="LOCTEXT(\"Setting TOD\")")
   > Setup Time Of Day!(execute:Exec) -> (then:Exec) {is=title K2Node_MacroInstance}
   > Capture Shot Macro!(
    execute:Exec
    Targets:Array of Texture 2D Object References=Get Targets_2k_Discovered() -> (Targets_2k_Discovered:Array of Texture 2D Object References)
    Resolution:int=2048
    SceneCapture:Scene Capture 2D Class Reference=Get SceneCapture_Discovered() -> (SceneCapture_Discovered:Scene Capture 2D Class Reference)
   ) -> (then:Exec) {is=title K2Node_MacroInstance}
   > Capture Shot Macro(
    Targets=Get Targets_4k_Discovered() -> (Targets_4k_Discovered:Array of Texture 2D Object References)
    Resolution=4096
    SceneCapture=Get SceneCapture_Discovered()
   )
   > Capture Shot Macro(
    Targets=Get Targets_8k_Discovered() -> (Targets_8k_Discovered:Array of Texture 2D Object References)
    Resolution=8192
    SceneCapture=Get SceneCapture_Discovered()
   )
   > Capture Shot Macro(
    Targets=Get Targets_2k_Undiscovered() -> (Targets_2k_Undiscovered:Array of Texture 2D Object References)
    Resolution=2048
    SceneCapture=Get SceneCapture_Undiscovered() -> (SceneCapture_Undiscovered:Scene Capture 2D Class Reference)
   )
   > Capture Shot Macro(
    Targets=Get Targets_4k_Undiscovered() -> (Targets_4k_Undiscovered:Array of Texture 2D Object References)
    Resolution=4096
    SceneCapture=Get SceneCapture_Undiscovered()
   )
   > FinishExecutingTask()
  ]}
 Execute Game World Console Command {nodes [
   n1 = Execute Game World Console Command!() -> (then, Command:String) {is=title K2Node_FunctionEntry}
   > n2 = GetGameWorld(self=Get UnrealEditorSubsystem())
   > ExecuteConsoleCommand(WorldContextObject=@n2, Command=@n1)
  ]}
 ExecuteGameWorldCommands {nodes [
   n1 = ExecuteGameWorldCommands!() -> (then, Commands:Array of Strings) {is=title K2Node_FunctionEntry}
   > n2 = For Each Loop(Array:Array of Strings=@n1) -> (Array Element:String)
   Return Node(execute=@n2.Completed)
   > Execute Game World Console Command!(execute, self, Command:String=@n2) -> (then) {is=title K2Node_CallFunction}
  ]}
 Find Editor Actor By Name {nodes [
   n1 = Find Editor Actor By Name!() -> (then, Name:String) {is=title K2Node_FunctionEntry}
   > n2 = GetAllLevelActors(self=Get EditorActorSubsystem())
   > n3 = For Each Loop(Array:Array of Actor Object References=@n2) -> (Array Element:Actor Object Reference)
   Return Node!(execute=@n3.Completed, Actor:Actor Object Reference)
   > Branch(Condition="Equal Exactly (String)"(A=GetDisplayName(Object=@n3), B=@n1))
   > Return Node(Actor=@n3)
  ]}
 Hide And Delete Actor {nodes [
   n1 = Hide And Delete Actor!() -> (then, Actor:Actor Object Reference) {is=title K2Node_FunctionEntry}
   > Set Actor Hidden In Game!(execute, self:Actor Object Reference=@n1, bNewHidden:bool=true) -> (then) {is=title K2Node_CallFunction}
   > Destroy Actor(self=@n1)
  ]}
 LoadAndSimulateMapMacro {nodes [
   n1 = AsyncEditorDelay(Seconds=1, MinimumFrames=30)
   n2 = Branch(execute=@n1.Complete, Condition=false)
   Set Progress(execute=[@n4.then, @n2.then], Text="LOCTEXT(\"Simulate Starting\")")
   > EditorPlaySimulate(self:Level Editor Subsystem Object Reference=Get LevelEditorSubsystem())
   > Set Progress(Text="LOCTEXT(\"Simulate Started\")")
   n3 = AsyncEditorOpenMapAndFocusActor!(execute:Exec, Map:Soft Object Path Structure=Get Map() -> (Map:Soft Object Path Structure), FocusActorName:String=MinimapCaptureCamera) -> (then:Exec, Complete:Exec) {is=title K2Node_AsyncAction}
   n4 = AutomationWaitForLoading!(execute=@n3.Complete, WorldContextObject:Object Reference, Options:Automation Wait For Loading Options Structure) -> (then)
   Inputs!() -> (execute)
   > Branch()
   > Wait for Loading!(execute:Exec) -> (then:Exec) {is=title K2Node_MacroInstance}
   > Outputs!(then:Exec)
   "Equal Exactly (String)"(A=BreakSoftObjectPath(InSoftObjectPath=Get Map()))
   EditorEndPlay!(execute) -> (then)
   Set Progress(Text="LOCTEXT(\"Loading Map...\")")
   Get PreserveExistingState() -> (PreserveExistingState:bool)
  ]}
 Set Progress {nodes [
   n1 = Set Progress!() -> (then, Text:Text) {is=title K2Node_FunctionEntry}
   > PrintText(InText=@n1)
   > SetTaskNotificationText!(execute, self:Editor Utility Task Object Reference, Text:"Text (by ref)"=@n1) -> (then)
  ]}
 SetupConsoleVariables {nodes [
   Inputs()
   > Set Progress(Text="LOCTEXT(\"Console Vars\")") {is=title K2Node_CallFunction}
   > ExecuteGameWorldCommands!(execute, self, Commands:Array of Strings=Get PreCaptureConsoleCommands() -> (PreCaptureConsoleCommands:Array of Strings)) -> (then) {is=title K2Node_CallFunction}
   > Outputs()
  ]}
 SetupTimeOfDay {nodes [
   Inputs()
   > Execute Game World Console Command(Command=cheat SetTimeOfDayAbsoluteSpeed 0.01)
   > Execute Game World Console Command(Command=cheat settimeofday 13)
   > Delay!(execute, WorldContextObject:Object Reference, Duration=3) -> (then)
   > Execute Game World Console Command(Command=slomo 0)
   > Execute Game World Console Command(Command=Cheat slomo 0)
   > Outputs()
  ]}
 WaitForLoading {nodes [
   n1 = AsyncWaitForGameWorld!(execute:Exec=@n3.then, Index:int=0, Server:bool=false) -> (then:Exec, Complete:Exec, World:World Object Reference) {is=title K2Node_AsyncAction}
   n2 = AutomationWaitForLoading(
    WorldContextObject=@n1
    Options=Make Automation Wait for Loading Options!(WaitForReplicationToSettle:bool=true) -> (AutomationWaitForLoadingOptions:Automation Wait For Loading Options Structure) {is=title K2Node_MakeStruct}
   )
   Set Progress(execute=[@n2.then, @n4.then], Text="LOCTEXT(\"Loading Done\")")
   > Outputs()
   Inputs()
   > n3 = Set Progress(Text="LOCTEXT(\"Waiting for Game World\")")
   n4 = Set Progress(execute=@n1.Complete, Text="LOCTEXT(\"Loading...\")")
  ]}
]
variables [
 LoadEverything {type=boolean, default=true}
 LoadMapOnly {type=boolean}
 Map {type=SoftObjectPath, default=""}
 PreCaptureConsoleCommands {
  type="array<string>"
  default [
   ShowHUD
   r.ViewDistanceScale 3000000
   r.Water.WaterMesh.LODScaleBias 99
   r.postprocessAAquality 0
   wp.Runtime.HLOD 0
   r.forcelod 0
   r.streaming.poolsize 2048
  ]
 }
 PreserveExistingState {type=boolean}
 Resolution {type=integer, default=4096}
 Scene Capture Class {type=Class, default=&/SceneCapture_Discovered.SceneCapture_Discovered_C}
 SceneCapture_Discovered {type=Class, default=&SceneCapture_Discovered_C}
 SceneCapture_Undiscovered {type=Class, default=&/SceneCapture_Undiscovered.SceneCapture_Undiscovered_C}
 Targets_2k_Discovered {type="array<Texture2D>", default [/Game/Athena/Apollo/Maps/UI/Automated_Captures/Minimap_2048_Automated]}
 Targets_2k_Undiscovered {type="array<Texture2D>", default [/Game/Athena/Apollo/Maps/UI/Automated_Captures/Minimap_Undiscovered_2048_Automated]}
 Targets_4k_Discovered {type="array<Texture2D>", default [/Game/Athena/Apollo/Maps/UI/Automated_Captures/Minimap_4096_Automated]}
 Targets_4k_Undiscovered {type="array<Texture2D>", default [/Game/Athena/Apollo/Maps/UI/Automated_Captures/Minimap_Undiscovered_4096_Automated]}
 Targets_8k_Discovered {type="array<Texture2D>", default [/Game/Athena/Apollo/Maps/UI/Automated_Captures/Minimap_8192_Automated]}
 WarmUpFrames {type=integer}
]
